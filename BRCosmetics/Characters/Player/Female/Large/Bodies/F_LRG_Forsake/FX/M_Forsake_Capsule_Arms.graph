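# Translucent unlit Niagara material for the capsule-arm tube effect on the Forsake character.
# Color blends a banded, hue-shifted two-color gradient or particle color, remapped to HDR off mobile, then adjusts it by time of day.
# Opacity combines UV-scaled masking, optional segmentation from two offset-panned texture samples, noise dissolve and particle alpha, with depth fade off mobile.
# Dynamic parameters drive offsets, dissolve, hue and a scrolling-noise wobble intended for WPO.
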Material /BRCosmetics/Characters/Player/Female/Large/Bodies/F_LRG_Forsake/FX/M_Forsake_Capsule_Arms
graph {
 properties {
  bEnableResponsiveAA=true
  bUsedWithNiagaraMeshParticles=true
  bUsedWithNiagaraRibbons=true
  bUsedWithNiagaraSprites=true
  blendMode=BLEND_Translucent
  pixelDepthOffsetMode=PDOM_Legacy
  shadingModel=MSM_Unlit
  usedShadingModels=Unlit
 }
 nodes [
  n1 = TextureCoordinate() {coordinateIndex=1}
  Add(A=@n1, B=NamedRerouteUsage())
  Add()
  n2 = VectorParameter() {group=Color, parameterName=Time Of Day Scalar, channelNames {a=Night, b=Evening, g=Day, r=Morning}, defaultValue=<rgba 1 1 1 1>}
  n3 = MaterialFunctionCall(Value_Morning=@n2.Morning, Value_Day=@n2.Morning, Value_Evening=@n2.Morning, Value_Night=@n2.Morning) {materialFunction=&Scalar_Time_of_Day}
  n4 = DynamicParameter() {defaultValue=<rgba -0.25 0.25 0 0>, paramNames [Offset 1, Offset 2, dissolve Amount, Hueshift]}
  n5 = LinearInterpolate(
   A=VectorParameter() {group=Color, parameterName=Outer Color, defaultValue=<rgba 6.9333 0 8 1>}.RGB
   B=VectorParameter() {group=Color, parameterName=inner color, defaultValue=<rgba 0.9333 0.5 1 1>}.RGB
   Alpha=MaterialFunctionCall!(
    Width=ScalarParameter() {defaultValue=0.5, group=Color, parameterName=Color Band width}
    Sharpness=ScalarParameter() {group=Color, parameterName=Color Band sharpness}
    Offset=ScalarParameter() {group=Color, parameterName="Color band Offset (S)"}
    Direction Switch=StaticBool() {value=true}
    Compare
    Input Coordinates
   ) {materialFunction=/Engine/Functions/Engine_MaterialFunctions03/Procedurals/GeneratedBand}.Result
  )
  n6 = LinearInterpolate(
   A=ParticleColor().RGB
   B=Multiply(
    A=MaterialFunctionCall(Hue Shift Percentage=@n4.Hueshift, Texture=@n5) {materialFunction=&HueShift}.Result
    B=DynamicParameter() {parameterIndex=2, defaultValue=<rgba 0 1 0 0>, paramNames [Masl Offset 1, Color Multiply, "", ""]}.Color Multiply
   )
   Alpha=ScalarParameter() {group=Color, parameterName=Use Param Color}
  )
  n7 = MaterialFunctionCall(
   LDR Input=Clamp(None=@n6)
   HDR Dynamic Range=Constant() {r=100}
   HDR Tint=@n6
   HDR Intensity Multiplier=ScalarParameter() {defaultValue=1, group=Derive HDR from LDR, parameterName=HDR Multiply}
   Derive HDR Power=ScalarParameter() {defaultValue=1, group=Derive HDR from LDR, parameterName="Derive HDR Power (S)"}
   Desaturate Luminance=ScalarParameter() {defaultValue=1, group=Derive HDR from LDR, parameterName="Desaturate Luminance (S)"}
  ) {materialFunction=&DeriveHDRfromLDR}
  n8 = Constant2Vector() {g=1, r=2}
  n9 = TextureSampleParameter2D(UVs=LinearInterpolate(
    A=@n1
    B=MaterialFunctionCall(UVs=@n1, Texture Scale=@n8) {materialFunction=&ScaleUVsByCenter}.UVs
    Alpha=ScalarParameter() {defaultValue=1, group=Masking, parameterName=Use Scale UV}
   )) {
   group=Masking
   parameterName=Mask
   samplerType=SAMPLERTYPE_LinearColor
   texture=&/T_ForSakeCharacterTubes_basecolor
  }
  n10 = ChannelMaskParameter(None=@n9.RGB) {group=Masking, parameterName=Texture Channel 2 overall mask}
  n11 = Panner(
   Coordinate=TextureCoordinate()
   Speed=AppendVector(A=ScalarParameter() {group=Texture, parameterName=Pan Speed X}, B=ScalarParameter() {group=Texture, parameterName=Pan Speed Y})
  )
  n12 = Panner(Coordinate=@n11, Time=@n4.Offset 1) {speedX=1}
  n13 = ScalarParameter() {defaultValue=1, group=Texture, parameterName=Clamp UV}
  n14 = TextureObjectParameter() {
   group=Texture
   parameterName=Texture
   samplerType=SAMPLERTYPE_LinearColor
   texture=&T_ForSakeCharacterTubes_basecolor
  }
  n15 = TextureSample(UVs=LinearInterpolate(A=@n12, B=Saturate(None=@n12), Alpha=@n13), Tex=@n14) {
   samplerType=SAMPLERTYPE_LinearColor
   texture=&T_ForSakeCharacterTubes_basecolor
  }
  n16 = Panner(Coordinate=@n11, Time=@n4.Offset 2) {speedX=1}
  n17 = TextureSample(UVs=LinearInterpolate(A=@n16, B=Saturate(None=@n16), Alpha=@n13), Tex=@n14) {
   samplerType=SAMPLERTYPE_LinearColor
   texture=&T_ForSakeCharacterTubes_basecolor
  }
  n18 = ChannelMaskParameter(None=Max(A=@n15.RGBA, B=@n17.RGBA)) {group=Texture, parameterName=Texture Channel}
  n19 = Saturate(None=LinearInterpolate(A=@n10.R, B=Multiply(A=@n18.R, B=@n10.R), Alpha=ScalarParameter() {group=Masking, parameterName=Use particle segmentation}))
  n20 = ScalarParameter() {defaultValue=5, group=Color, parameterName=texture mask MulExp}
  n21 = Multiply(
   A=ShadingPathSwitch(Default=@n7.RemappedRange, Deferred=@n7.RemappedRange, Forward=@n7.RemappedRange, Mobile=@n6)
   B=LinearInterpolate(B=Power(Base=@n19, Exp=@n20) {constExponent=5}, Alpha=ScalarParameter() {defaultValue=0.5, group=Color, parameterName=Use Texture Mask Mul}) {constA=1}
  )
  n22 = VectorParameter() {group=Color, parameterName=Time Of Day desaturation, channelNames {a=Night, b=Evening, g=Day, r=Morning}}
  MaterialFunctionCall(Emissive=Desaturation(
    None=Multiply(A=@n3.Result, B=@n21)
    Fraction=MaterialFunctionCall(Value_Morning=@n22.Morning, Value_Day=@n22.Morning, Value_Evening=@n22.Morning, Value_Night=@n22.Morning) {materialFunction=&Scalar_Time_of_Day}.Result
   )) {materialFunction=&MF_FX_EmissiveAdjustments}
  n23 = DynamicParameter() {parameterIndex=2, defaultValue=<rgba 0 1 0 0>, paramNames [Masl Offset 1, Color Multiply, "", ""]}
  n24 = Constant()
  n25 = TextureSampleParameter2D(UVs=Add(A=TextureCoordinate() {coordinateIndex=1}, B=AppendVector(A=@n23.Masl Offset 1, B=@n24))) {
   group=Masking
   parameterName=Mask 2
   samplerSource=SSM_Clamp_WorldGroupSettings
   texture=&/T_ForSakeCharacterTubes_Mask
  }
  Min(A=@n9.RGB, B=LinearInterpolate(B=@n25.RGB, Alpha=ScalarParameter() {group=Masking, parameterName=Use Mask 2}) {constA=1})
  n26 = Add(A=Time(), B=ParticleRandom())
  n27 = Panner(
   Coordinate=Multiply(A=TextureCoordinate(), B=ScalarParameter() {defaultValue=1, group=Wpo, parameterName=WPO TIle})
   Time=@n26
   Speed=ScalarParameter() {defaultValue=0.25, group=Wpo, parameterName=WPO Speed X}
  ) {speedY=0.5}
  n28 = ScalarParameter() {defaultValue=0.25, group=Wpo, parameterName=WPO Speed Y}
  n29 = TextureSampleParameter2D(UVs=AppendVector(
    A=MaterialFunctionCall(UV Channel=@n27) {materialFunction=&LinearGradient}.UGradient
    B=Multiply(A=@n26, B=@n28)
   )) {
   group=Wpo
   parameterName=WPO Noise Texture
   samplerType=SAMPLERTYPE_LinearColor
   texture=&HighContrastNoise
  }
  Multiply(
   A=Subtract(A=@n29.RGB) {constB=0.5}
   B=DynamicParameter() {parameterIndex=1, defaultValue=<rgba -0.25 0.25 1 0>, paramNames [temp wpo Str, NA, NA, NA]}.temp wpo Str
  ) {constB=2}
  ParticleColor()
  n30 = Power(Base=OneMinus(None=@n19))
  n31 = VectorParameter() {group=dissolve, parameterName=Dissove Tex Options, channelNames {a=Speed Y, b=Speed X, g=Tile Y, r=TIle X}, defaultValue=<rgba 1 1 0 0>}
  n32 = AppendVector(A=@n31.TIle X, B=@n31.TIle X)
  n33 = TextureSample(UVs=Panner(Coordinate=Multiply(A=@n11, B=@n32), Speed=AppendVector(A=@n31.Speed X, B=@n31.Speed X))) {
   samplerType=SAMPLERTYPE_LinearGrayscale
   texture=/BRCosmetics/Athena/Effects/Particles/Skydiving/NS_Contrail_DarkAndLight/T_DarkAndLight_Noise
  }
  n34 = Add(A=LinearInterpolate(Alpha=@n4.dissolve Amount) {constA=1, constB=-1}, B=@n33.R)
  n35 = Multiply(A=ParticleColor().A, B=Multiply(A=OneMinus(None=@n30), B=Saturate(None=@n34)))
  ShadingPathSwitch(Default=DepthFade(Opacity=@n35, FadeDistance=ScalarParameter() {group=Opacity, parameterName=FadeDistance}), Mobile=@n35)
 ]
}
