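# Niagara system for the F_LRG_Forsake character body cosmetic, adding flickering electric ribbons and small glowing sprites along the skeleton.
# FN_BoneFlow wraps particles between sampled bones or sockets, with curve-driven radius and partitioned bursts forming looping ribbons.
# An alt-skin bool switches colors, sizes and intensity, with further scaling for front end versus gameplay and time of day.
# Platform quality sets choose per-emitter particle counts, which determine ribbon point density.
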NiagaraSystem /BRCosmetics/Characters/Player/Female/Large/Bodies/F_LRG_Forsake/FX/NS_Forsake_Character
dependencies {
 usedDataInterfaces [
  &NiagaraDataInterfacePlatformSet
  &NiagaraDataInterfaceCurve
  &NiagaraDataInterfaceVectorCurve
  &NiagaraDataInterfaceArrayInt32
  &NiagaraDataInterfaceArrayFloat3
 ]
 usedDynamicInputs [&Multiply_Vector, &MakeCustomVectorFromBool]
 usedModules [
  &self:SystemSpawnScript.SystemScriptSource.SystemScriptGraph.NiagaraNodeAssignment_0.SetVariables_#1
  &SystemState
 ]
}
emitters {
 CakeSprinkles {
  emitterSpawnScript [
   /SetVariables(Emitter.NumParticles:NiagaraInt32=&MakeCustomIntFromBool(
     Bool=&IsPlatformSetActive(New Platform Set=DataInterface() {propertyValues {Platforms {qualityLevelMask=-31}}})
     True INT=3
     False INT=&MakeCustomIntFromBool(
      Bool=&IsPlatformSetActive(New Platform Set=DataInterface() {propertyValues {Platforms {qualityLevelMask=-30}}})
      True INT=5
      False INT=&MakeCustomIntFromBool(
       Bool=&IsPlatformSetActive(New Platform Set=DataInterface() {propertyValues {Platforms {qualityLevelMask=-28}}})
       True INT=7
       False INT=9
      )
     )
    )) {isSetParameters=true, module=SetVariables_#2}
   /SetVariables(Emitter.NewScale:NiagaraFloat=&Divide_Float(
     A=1
     B=&OneMinusFloat(Float=&Divide_Float(A=1, B=&MakeFloatFromInt(Int=$Emitter.NumParticles)))
    )) {isSetParameters=true, module=SetVariables_#3}
  ]
  emitterUpdateScript [
   &EmitterState(Life Cycle Mode=::NewEnumerator1, UseLoopDelay=false)
   &SpawnRate(SpawnRate=10, Use Spawn Probability=false)
  ]
  particleSpawnScript [
   &InitializeParticle(
    Color=&MakeCustomLinearColorFromBool(
     Bool=$System.is alt bool
     True Color=&LerpLinearColors(
      StartColor=<rgba 0.525 0.525 1.5 1>
      EndColor=<rgba 0.35 0.35 1 1>
      LerpFactor=&FN_FrontEndFloatModifier(Gameplay=0, Front End=1)
     )
     False Color=<rgba 1.33 0.225 1.5 1>
    )
    Lifetime Max=0.75
    Lifetime Min=0.25
    Lifetime Mode=::NewEnumerator1
    Sprite Rotation Mode=::NewEnumerator2
    Sprite Size=&Multiply_Vector2D_ByFloat(
     Float=&MakeCustomFloatFromBool(Bool=$System.is alt bool, True Float=1.75, False Float=1.5)
     Vector2D=&RandomRangeVector2D(Minimum=<3 1.5>, Maximum=<1.5 1>)
    )
    Sprite Size Mode=::NewEnumerator4
    UsePositionOffset=false
    Write Material Random=true
   )
   /SetVariables(Particles.Velocity=&RandomVector(Vector Scale=1)) {isSetParameters=true}
   /SetVariables(Particles.LeftOr Right:NiagaraInt32=&RandomRangeInt(Minimum=0, Maximum=1)) {isSetParameters=true}
   /SetVariables(Particles.BF Height:NiagaraFloat=&RandomRangeFloat(Minimum=0, Maximum=1)) {isSetParameters=true}
   &DynamicMaterialParameters(
    Module.Param0WriteEnabled=true
    Module.Param0WriteEnabled1=true
    Module.Param1WriteEnabled=true
    Module.Param1WriteEnabled1=true
    Module.Param2WriteEnabled=true
    Module.Param2WriteEnabled1=true
    Module.Param3WriteEnabled=true
    Module.Param3WriteEnabled1=true
    Write Parameter Index 1=true
   )
  ]
  particleUpdateScript [
   &ParticleState()
   &SampleSkeletalMesh(
    AdditionalYawPitchRoll=false
    Bone Sampling Mode:ENiagaraBoneSamplingMode=::NewEnumerator4
    Skeletal Mesh:NiagaraDataInterfaceSkeletalMesh=$User.SkeletalMesh
   )
   &SampleSkeletalMesh(
    AdditionalYawPitchRoll=false
    Bone / Socket Index=1
    Bone Sampling Mode:ENiagaraBoneSamplingMode=::NewEnumerator4
    Skeletal Mesh:NiagaraDataInterfaceSkeletalMesh=$User.SkeletalMesh
   ) {module=SampleSkeletalMesh003}
   &SampleSkeletalMesh(
    AdditionalYawPitchRoll=false
    Bone / Socket Index=3
    Bone Sampling Mode:ENiagaraBoneSamplingMode=::NewEnumerator4
    Skeletal Mesh:NiagaraDataInterfaceSkeletalMesh=$User.SkeletalMesh
   ) {module=SampleSkeletalMesh002}
   &SampleSkeletalMesh(
    AdditionalYawPitchRoll=false
    Bone / Socket Index=4
    Bone Sampling Mode:ENiagaraBoneSamplingMode=::NewEnumerator4
    Skeletal Mesh:NiagaraDataInterfaceSkeletalMesh=$User.SkeletalMesh
   ) {module=SampleSkeletalMesh001}
   &FN_BoneFlow(
    Angle:NiagaraFloat=&RandomRangeFloat(Minimum=0, Maximum=1)
    BoneID:NiagaraInt32=$Particles.LeftOr Right
    Height:NiagaraFloat=&RandomRangeFloat(Minimum=-0.1, Maximum=0.85)
    Radius:NiagaraFloat=&MakeCustomFloatFromBool(
     Bool=&SetBoolByIntComparison(A=$Particles.LeftOr Right, B=0, Comparison Type=::NewEnumerator0)
     True Float=&FloatFromCurve(
      FloatCurve=DataInterface() {propertyValues {Curve {keys [
          {
           arriveTangent=0.0989
           arriveTangentWeight=0.8107
           interpMode=RCIM_Cubic
           leaveTangent=0.0989
           leaveTangentWeight=0.8107
           tangentMode=RCTM_User
           tangentWeightMode=RCTWM_WeightedBoth
           time=0
           value=1
          }
          {
           arriveTangent=-0.5593
           arriveTangentWeight=0.5401
           interpMode=RCIM_Cubic
           leaveTangent=-0.5593
           leaveTangentWeight=0.5401
           tangentMode=RCTM_User
           tangentWeightMode=RCTWM_WeightedBoth
           time=1
           value=0
          }
         ]}}}
      CurveIndex=$Particles.BF Height
      Scale Curve=7
     )
     False Float=&FloatFromCurve(
      FloatCurve=DataInterface() {propertyValues {Curve {keys [
          {
           arriveTangent=1.2538
           arriveTangentWeight=0.3265
           interpMode=RCIM_Cubic
           leaveTangent=1.2538
           leaveTangentWeight=0.3265
           tangentMode=RCTM_User
           tangentWeightMode=RCTWM_WeightedBoth
           time=0
           value=0.5
          }
          {
           arriveTangent=-0.5753
           arriveTangentWeight=0.2828
           interpMode=RCIM_Cubic
           leaveTangent=-0.5753
           leaveTangentWeight=0.1886
           tangentMode=RCTM_User
           time=0.6639
           value=0.7732
          }
          {
           arriveTangent=-1.0575
           arriveTangentWeight=0.3694
           interpMode=RCIM_Cubic
           leaveTangent=-1.0575
           leaveTangentWeight=0.3694
           tangentMode=RCTM_User
           tangentWeightMode=RCTWM_WeightedBoth
           time=1
           value=0
          }
         ]}}}
      CurveIndex=$Particles.BF Height
      Scale Curve=6
     )
    )
    SkeletalMesh:NiagaraDataInterfaceSkeletalMesh=$User.SkeletalMesh
   )
   &ScaleColor(
    Scale Alpha=&FloatFromCurve(FloatCurve=DataInterface() {propertyValues {Curve {keys [
         {arriveTangent=5.4725, interpMode=RCIM_Cubic, time=0, value=0}
         {interpMode=RCIM_Cubic, tangentMode=RCTM_User, time=0.5, value=1}
         {interpMode=RCIM_Cubic, leaveTangent=-6.0486, time=1, value=0}
        ]}}})
    Scale RGB=&Multiply_VectorByFloat(
     Vector=&VectorFromFloat(Value=&MakeCustomFloatFromBool(
       Bool=$System.is alt bool
       True Float=&FN_FrontEndFloatModifier(
        Gameplay=&Lerp_Float(
         A=0.4
         B=0.5
         Alpha=&Sine(
          Normalized Angle=&Add_Float(A=$Emitter.Age, B=&RandomRangeFloat(Minimum=0, Maximum=1))
          Period=0.1
          Scale=0.5
          Bias=0.5
         )
        )
        Front End=&Lerp_Float(
         A=0.8
         B=1
         Alpha=&Sine(
          Normalized Angle=&Add_Float(A=$Emitter.Age, B=&RandomRangeFloat(Minimum=0, Maximum=1))
          Period=0.1
          Scale=0.5
          Bias=0.5
         )
        )
       )
       False Float=&FN_FrontEndFloatModifier(
        Gameplay=&Lerp_Float(
         A=0.65
         B=1
         Alpha=&Sine(
          Normalized Angle=&Add_Float(A=$Emitter.Age, B=&RandomRangeFloat(Minimum=0, Maximum=1))
          Period=0.1
          Scale=0.5
          Bias=0.5
         )
        )
        Front End=&Lerp_Float(
         A=1.5
         B=2
         Alpha=&Sine(
          Normalized Angle=&Add_Float(A=$Emitter.Age, B=&RandomRangeFloat(Minimum=0, Maximum=1))
          Period=0.1
          Scale=0.5
          Bias=0.5
         )
        )
       )
      ))
     Float=&MakeCustomFloatFromBool(
      Bool=$System.is alt bool
      True Float=&FNTimeOfDayScalar(Night Scale=0.65, Morning Scale=0.25, Evening Scale=0.5)
      False Float=&FNTimeOfDayScalar(Night Scale=0.65, Morning Scale=0.5, Evening Scale=0.5)
     )
    )
    ScaleA=true
    ScaleRGB=true
   )
   &ShapeLocation(
    MaskSpawnGroup=false
    Override Seed=false
    Shape Origin=$Particles.Position
    Sphere Radius=&FloatFromCurve(
     FloatCurve=DataInterface() {propertyValues {Curve {keys [{tangentMode=RCTM_User, time=0, value=0}, {tangentMode=RCTM_User, time=1, value=1}]}}}
     Scale Curve=0.35
    )
    UseOwnerScale=true
   ) {module=ShapeLocation001}
   &SpriteRotationRate(Rotation Rate=&RandomRangeFloat(Minimum=-30, Maximum=30))
   &DynamicMaterialParameters(Module.Param0WriteEnabled=true, Module.Param1WriteEnabled=true, Module.Param2WriteEnabled=true, Module.Param3WriteEnabled=true) {module=DynamicMaterialParameters001}
   /SetVariables(Particles.Position=&Add_Vector(A=$Particles.ShapeLocation001.ShapePosition, B=$Particles.FN_BoneFlow.Position)) {isSetParameters=true, module=SetVariables_#4}
  ]
  rendererClasses [&NiagaraSpriteRendererProperties]
  renderers [{rendererClass=&NiagaraSpriteRendererProperties}]
 }
 Electricity_MainLine {
  emitterSpawnScript [
   /SetVariables(Emitter.NumParticles:NiagaraInt32=&MakeCustomIntFromBool(
     Bool=&IsPlatformSetActive(New Platform Set=DataInterface() {propertyValues {Platforms {qualityLevelMask=-31}}})
     True INT=7
     False INT=&MakeCustomIntFromBool(
      Bool=&IsPlatformSetActive(New Platform Set=DataInterface() {propertyValues {Platforms {qualityLevelMask=-30}}})
      True INT=7
      False INT=&MakeCustomIntFromBool(
       Bool=&IsPlatformSetActive(New Platform Set=DataInterface() {propertyValues {Platforms {qualityLevelMask=-28}}})
       True INT=13
       False INT=13
      )
     )
    )) {isSetParameters=true, module=SetVariables_#2}
   /SetVariables(Emitter.NewScale:NiagaraFloat=&Divide_Float(
     A=1
     B=&OneMinusFloat(Float=&Divide_Float(A=1, B=&MakeFloatFromInt(Int=$Emitter.NumParticles)))
    )) {isSetParameters=true, module=SetVariables_#3}
   /SetVariables(Emitter.IsRegularSKin:NiagaraBool=&Boolean_NotOperation(Bool=$System.is alt bool)) {isSetParameters=true}
   /SetVariables(Emitter.IsShardrSKin:NiagaraBool=$System.is alt bool) {isSetParameters=true}
  ]
  emitterUpdateScript [
   /SetVariables(Emitter.Life:NiagaraFloat=&Lerp_Float(
     A=0.75
     B=1
     Alpha=&FloatFromCurve(
      FloatCurve=DataInterface() {propertyValues {Curve {keys [
          {
           arriveTangent=1.2434
           arriveTangentWeight=0.5979
           interpMode=RCIM_Cubic
           leaveTangent=1.2434
           leaveTangentWeight=0.5979
           tangentMode=RCTM_User
           tangentWeightMode=RCTWM_WeightedBoth
           time=0
           value=0
          }
          {
           arriveTangentWeight=0.8985
           interpMode=RCIM_Cubic
           leaveTangentWeight=0.8985
           tangentMode=RCTM_User
           tangentWeightMode=RCTWM_WeightedBoth
           time=1
           value=1
          }
         ]}}}
      CurveIndex=&RandomRangeFloat(Minimum=0, Maximum=1)
     )
    )) {isSetParameters=true, module=SetVariables_#5}
   /SetVariables(Emitter.Life:NiagaraFloat=&Lerp_Float(
     A=0.075
     B=1
     Alpha=&FloatFromCurve(
      FloatCurve=DataInterface() {propertyValues {Curve {keys [
          {
           arriveTangent=1.2434
           arriveTangentWeight=0.5979
           interpMode=RCIM_Cubic
           leaveTangent=1.2434
           leaveTangentWeight=0.5979
           tangentMode=RCTM_User
           tangentWeightMode=RCTWM_WeightedBoth
           time=0
           value=0
          }
          {
           arriveTangentWeight=0.8985
           interpMode=RCIM_Cubic
           leaveTangentWeight=0.8985
           tangentMode=RCTM_User
           tangentWeightMode=RCTWM_WeightedBoth
           time=1
           value=1
          }
         ]}}}
      CurveIndex=&RandomRangeFloat(Minimum=0, Maximum=1)
     )
    )) {enabled=false, isSetParameters=true, module=SetVariables_#6}
   &EmitterState(Life Cycle Mode=::NewEnumerator1, Loop Duration=$Emitter.Life, Recalculate Duration Each Loop=true, UseLoopDelay=false)
   &SpawnBurst_Instantaneous(Spawn Count=$Emitter.NumParticles, Use Spawn Probability=false, UseLoopCountLimit=false)
   &SpawnBurst_Instantaneous(Spawn Count=$Emitter.NumParticles, Spawn Group=1, Use Spawn Probability=false, UseLoopCountLimit=false) {enabled=false, module=SpawnBurst_Instantaneous001}
   /SetVariables(Emitter.RandomRibbonOffset:NiagaraFloat=&RandomRangeFloat(Minimum=0, Maximum=1, Evaluation Type=::NewEnumerator1)) {isSetParameters=true, module=SetVariables_#7}
   /SetVariables(
    Emitter.random boneflow Angle Left:NiagaraFloat=&RandomRangeFloat(Minimum=0, Maximum=0.3333)
    Emitter.random boneflow Angle Right:NiagaraFloat=&RandomRangeFloat(Minimum=0, Maximum=0.3333, Evaluation Type=::NewEnumerator1)
   ) {isSetParameters=true, module=SetVariables_#8}
   /SetVariables(Emitter.Ribbon Life:NiagaraFloat=&RandomRangeFloat(Minimum=1.5, Maximum=2.5, Evaluation Type=::NewEnumerator1)) {isSetParameters=true, module=SetVariables_#9}
   /SetVariables(Emitter.Is Broken ribbon:NiagaraBool=&SetBoolByFloatComparison(A=&RandomRangeFloat(Minimum=0, Maximum=1, Evaluation Type=::NewEnumerator1), B=0.5)) {isSetParameters=true, module=SetVariables_#10}
   /SetVariables(
    Emitter.Random Ribbon Height Left:NiagaraFloat=&RandomRangeFloat(Minimum=0.001, Maximum=0.1, Evaluation Type=::NewEnumerator1)
    Emitter.Random Ribbon Height Right:NiagaraFloat=&RandomRangeFloat(Minimum=0.001, Maximum=0.1, Evaluation Type=::NewEnumerator1)
   ) {isSetParameters=true, module=SetVariables_#11}
   /SetVariables(
    Emitter.Random Ribbon Height Left 2:NiagaraFloat=&RandomRangeFloat(Minimum=9, Maximum=0.9, Evaluation Type=::NewEnumerator1)
    Emitter.Random Ribbon Height Right 2:NiagaraFloat=&RandomRangeFloat(Minimum=0.9, Maximum=0.9, Evaluation Type=::NewEnumerator1)
   ) {isSetParameters=true, module=SetVariables_#12}
   /SetVariables(
    Emitter.UV Offset 1:NiagaraFloat=&RandomRangeFloat(Minimum=-1, Maximum=1, Evaluation Type=::NewEnumerator1)
    Emitter.UV Offset 2:NiagaraFloat=&RandomRangeFloat(Minimum=0.75, Maximum=-0.75, Evaluation Type=::NewEnumerator1)
   ) {isSetParameters=true, module=SetVariables_#13}
   /SetVariables(Emitter.Material Random:NiagaraFloat=&RandomRangeFloat(Minimum=-100, Maximum=100, Evaluation Type=::NewEnumerator1)) {isSetParameters=true, module=SetVariables_#14}
   /SetVariables(Emitter.CutOutOfset:NiagaraFloat=&RandomRangeFloat(Minimum=-0.5, Maximum=0.5, Evaluation Type=::NewEnumerator1)) {isSetParameters=true, module=SetVariables_#15}
   /SetVariables(Emitter.CutOutOfset:NiagaraFloat=&RandomRangeFloat(Minimum=50, Maximum=50, Evaluation Type=::NewEnumerator1)) {isSetParameters=true, module=SetVariables_#16}
  ]
  particleSpawnScript [
   &InitializeRibbon(
    Color=&MakeCustomLinearColorFromBool(Bool=$System.is alt bool, True Color=<rgba 2 2 8 1>, False Color=<rgba 6.9333 0 8 1>)
    Lifetime=&Add_Float(A=$Emitter.Life, B=1.2)
    Material Random:NiagaraFloat=$Emitter.Material Random
    Ribbon Width=&MakeCustomFloatFromBool(
     Bool=$System.is alt bool
     True Float=&RandomRangeFloat(Minimum=3, Maximum=2.5)
     False Float=&RandomRangeFloat(Minimum=2, Maximum=2.5)
    )
    UsePositionOffset=false
    Write Material Random=true
   )
   &PartitionParticles(
    Particles Per Partition:NiagaraInt32=$Emitter.NumParticles
    Partition Count Mode:ENiagara_InfiniteLoopDuration=::NewEnumerator1
    Partition Mode:ENiagara_PartitionMode=::NewEnumerator1
    Write Partition Index to Ribbon Link Order:ENiagara_AttributeSamplingApplyOutput=::NewEnumerator0
    Write Partition to Ribbon ID=::NewEnumerator0
   )
   /SetVariables(Particles.GetRandomRibbonOffset:NiagaraFloat=$Emitter.RandomRibbonOffset) {isSetParameters=true, module=SetVariables_#17}
   /SetVariables(Particles.Get Random Boneflow Angle:NiagaraFloat=&MakeCustomFloatFromBool(
     Bool=&SetBoolByIntComparison(A=&ModuloInt(A=$Particles.PartitionParticles.Partition, B=2), B=0, Comparison Type=::NewEnumerator0)
     True Float=$Emitter.random boneflow Angle Left
     False Float=$Emitter.random boneflow Angle Right
    )) {isSetParameters=true, module=SetVariables_#18}
   /SetVariables(Particles.Get Is Broken ribbon:NiagaraBool=$Emitter.Is Broken ribbon) {isSetParameters=true, module=SetVariables_#19}
   /SetVariables(Particles.Get Random ribbon height:NiagaraFloat=&MakeCustomFloatFromBool(
     Bool=&SetBoolByIntComparison(A=&ModuloInt(A=$Particles.PartitionParticles.Partition, B=2), B=0, Comparison Type=::NewEnumerator0)
     True Float=$Emitter.Random Ribbon Height Left
     False Float=$Emitter.Random Ribbon Height Right
    )) {isSetParameters=true, module=SetVariables_#20}
   /SetVariables(Particles.Get Uv Offset 1:NiagaraFloat=$Emitter.UV Offset 1, Particles.Get Uv Offset 2:NiagaraFloat=$Emitter.UV Offset 2) {isSetParameters=true, module=SetVariables_#21}
   /SetVariables(Particles.GetcoutoutOffset:NiagaraFloat=$Emitter.CutOutOfset) {isSetParameters=true, module=SetVariables_#22}
  ]
  particleUpdateScript [
   &ParticleState()
   &SampleSkeletalMesh(
    AdditionalYawPitchRoll=false
    Bone Sampling Mode:ENiagaraBoneSamplingMode=::NewEnumerator4
    Skeletal Mesh:NiagaraDataInterfaceSkeletalMesh=$User.SkeletalMesh
   )
   &SampleSkeletalMesh(
    AdditionalYawPitchRoll=false
    Bone / Socket Index=1
    Bone Sampling Mode:ENiagaraBoneSamplingMode=::NewEnumerator4
    Skeletal Mesh:NiagaraDataInterfaceSkeletalMesh=$User.SkeletalMesh
   ) {module=SampleSkeletalMesh003}
   &FN_BoneFlow(
    Angle:NiagaraFloat=&Add_Float(
     A=$Particles.Get Random Boneflow Angle
     B=&Divide_Float(A=&MakeFloatFromInt(Int=$Particles.PartitionParticles.Partition), B=3)
    )
    BoneID=1
    Height:NiagaraFloat=&Lerp_Float(
     A=0
     B=1
     Alpha=&Multiply_Float(A=$Particles.PartitionParticles.Normalized Index In Partition, B=$Emitter.NewScale)
    )
    Radius:NiagaraFloat=&FloatFromCurve(
     FloatCurve=DataInterface() {propertyValues {Curve {keys [
         {
          arriveTangentWeight=0.3001
          interpMode=RCIM_Cubic
          leaveTangentWeight=0.3001
          tangentMode=RCTM_User
          tangentWeightMode=RCTWM_WeightedBoth
          time=0
          value=0.2
         }
         {arriveTangentWeight=0.2025, interpMode=RCIM_Cubic, leaveTangentWeight=0.1519, tangentMode=RCTM_User, time=0.4, value=1}
         {arriveTangentWeight=0.1133, interpMode=RCIM_Cubic, leaveTangentWeight=0.1417, tangentMode=RCTM_User, time=0.5623, value=0.4447}
         {arriveTangent=-2.809, interpMode=RCIM_Cubic, leaveTangent=-2.809, tangentMode=RCTM_User, time=1, value=0}
        ]}}}
     CurveIndex=&Multiply_Float(A=$Particles.PartitionParticles.Normalized Index In Partition, B=$Emitter.NewScale)
     Scale Curve=&RandomRangeFloat(Minimum=7, Maximum=6)
    )
    Set Material Dynamic=false
    SkeletalMesh:NiagaraDataInterfaceSkeletalMesh=$User.SkeletalMesh
    Use Sockets Instead=true
   )
   &DynamicMaterialParameters(
    Index 0 Param 1=$Particles.Get Uv Offset 1
    Index 0 Param 2=$Particles.Get Uv Offset 1
    Index 0 Param 3=&FloatFromCurve(
     FloatCurve=DataInterface() {propertyValues {Curve {keys [{tangentMode=RCTM_User, time=0.9, value=0}, {tangentMode=RCTM_User, time=1, value=1}]}}}
     Scale Curve=0
    )
    Index 0 Param 4=&Add_Float(
     A=&Add_Float(
      A=&MakeCustomFloatFromBool(Bool=$System.is alt bool, True Float=0, False Float=0)
      B=&MakeCustomFloatFromBool(
       Bool=$System.is alt bool
       True Float=&FloatFromCurve(
        FloatCurve=DataInterface() {propertyValues {Curve {keys [
            {arriveTangent=0.2, interpMode=RCIM_Cubic, time=0.1, value=-0.025}
            {arriveTangentWeight=0.2567, interpMode=RCIM_Cubic, leaveTangentWeight=0.2567, time=0.5, value=0}
            {
             arriveTangentWeight=0.4239
             interpMode=RCIM_Cubic
             leaveTangentWeight=0.4239
             tangentMode=RCTM_User
             tangentWeightMode=RCTWM_WeightedBoth
             time=1
             value=-0.025
            }
           ]}}}
        CurveIndex=&Multiply_Float(A=$Emitter.NewScale, B=$Particles.PartitionParticles.Normalized Index In Partition)
        Scale Curve=0
       )
       False Float=&FloatFromCurve(
        FloatCurve=DataInterface() {propertyValues {Curve {keys [
            {arriveTangent=0.2, interpMode=RCIM_Cubic, time=0.1, value=-0.025}
            {arriveTangentWeight=0.2567, interpMode=RCIM_Cubic, leaveTangentWeight=0.2567, time=0.5, value=0}
            {
             arriveTangentWeight=0.4239
             interpMode=RCIM_Cubic
             leaveTangentWeight=0.4239
             tangentMode=RCTM_User
             tangentWeightMode=RCTWM_WeightedBoth
             time=1
             value=-0.025
            }
           ]}}}
        CurveIndex=&Multiply_Float(A=$Emitter.NewScale, B=$Particles.PartitionParticles.Normalized Index In Partition)
        Scale Curve=0
       )
      )
     )
     B=&MakeCustomFloatFromBool(
      Bool=$System.is alt bool
      True Float=&RandomRangeFloat(Minimum=0, Maximum=0)
      False Float=&RandomRangeFloat(Minimum=0, Maximum=0)
     )
    )
    Index 1 Param 1=&Multiply_Float(
     A=&FloatFromCurve(
      FloatCurve=DataInterface() {propertyValues {Curve {keys [{tangentMode=RCTM_User, time=0, value=1}, {tangentMode=RCTM_User, time=1, value=1}]}}}
      Scale Curve=&RandomRangeFloat(Minimum=10, Maximum=10)
     )
     B=&FloatFromCurve(
      FloatCurve=DataInterface() {propertyValues {Curve {keys [
          {interpMode=RCIM_Cubic, tangentMode=RCTM_User, time=0, value=0}
          {interpMode=RCIM_Cubic, tangentMode=RCTM_User, time=0.25, value=1}
          {arriveTangentWeight=0.2778, interpMode=RCIM_Cubic, leaveTangentWeight=0.1389, tangentMode=RCTM_User, time=0.75, value=1}
          {tangentMode=RCTM_User, time=1, value=0}
         ]}}}
      CurveIndex=&Multiply_Float(A=$Particles.PartitionParticles.Normalized Index In Partition, B=$Emitter.NewScale)
      Scale Curve=0.66
     )
    )
    Index 2 Param 1=&Add_Float(A=$Particles.GetcoutoutOffset, B=&Multiply_Float(A=$Emitter.LoopedAge, B=-0.025))
    Index 2 Param 2=&MakeCustomFloatFromBool(
     Bool=$System.is alt bool
     True Float=&FN_FrontEndFloatModifier(
      Gameplay=&Lerp_Float(A=0.85, B=1, Alpha=&Sine(Normalized Angle=$Emitter.Age, Period=0.1, Scale=0.5, Bias=0.5))
      Front End=&Lerp_Float(A=1.35, B=1.5, Alpha=&Sine(Normalized Angle=$Emitter.Age, Period=0.1, Scale=0.5, Bias=0.5))
     )
     False Float=&FN_FrontEndFloatModifier(
      Gameplay=&Lerp_Float(A=1.5, B=2, Alpha=&Sine(Normalized Angle=$Emitter.Age, Period=0.1, Scale=0.5, Bias=0.5))
      Front End=&Lerp_Float(A=4, B=5, Alpha=&Sine(Normalized Angle=$Emitter.Age, Period=0.1, Scale=0.5, Bias=0.5))
     )
    )
    Index 2 Param 3=1
    Module.Param0WriteEnabled=true
    Module.Param0WriteEnabled1=true
    Module.Param0WriteEnabled2=true
    Module.Param1WriteEnabled=true
    Module.Param1WriteEnabled1=true
    Module.Param1WriteEnabled2=true
    Module.Param2WriteEnabled=true
    Module.Param2WriteEnabled1=true
    Module.Param2WriteEnabled2=true
    Module.Param3WriteEnabled=true
    Module.Param3WriteEnabled1=true
    Module.Param3WriteEnabled2=true
    Write Parameter Index 1=true
    Write Parameter Index 2=true
   )
   &ShapeLocation(
    Hemisphere Distribution=<1 0.5>
    MaskSpawnGroup=false
    Override Seed=false
    Shape Origin=$Particles.Position
    Sphere Radius=&Multiply_Float(
     A=&FloatFromCurve(
      FloatCurve=DataInterface() {propertyValues {Curve {keys [{tangentMode=RCTM_User, time=0, value=0.9}, {tangentMode=RCTM_User, time=1, value=1.25}]}}}
      Scale Curve=1.1
     )
     B=&FloatFromCurve(
      FloatCurve=DataInterface() {propertyValues {Curve {keys [
          {interpMode=RCIM_Cubic, tangentMode=RCTM_User, time=0.05, value=0}
          {interpMode=RCIM_Cubic, tangentMode=RCTM_User, time=0.25, value=1}
          {arriveTangentWeight=0.2778, interpMode=RCIM_Cubic, leaveTangentWeight=0.1389, tangentMode=RCTM_User, time=0.7742, value=1}
          {tangentMode=RCTM_User, time=0.95, value=0}
         ]}}}
      CurveIndex=&Multiply_Float(A=$Particles.PartitionParticles.Normalized Index In Partition, B=$Emitter.NewScale)
      Scale Curve=&RandomRangeFloat(Minimum=1, Maximum=1)
     )
    )
    UseOwnerScale=true
   )
   /SetVariables(
    Particles.GlowSpriteColor:LinearColor=&MultiplyLinearColor(
     Linear Color=$Particles.Initial.Color
     Scale Factor=&MakeFloat4FromVectorAndFloat(
      "Vector (RGB)"=<0.3 0.3 0.3>
      "Float (Alpha)"=&MakeCustomFloatFromBool(
       Bool=$System.is alt bool
       True Float=&FNTimeOfDayScalar(Night Scale=0.125, Morning Scale=0.15, Day Scale=0.2, Evening Scale=0.15)
       False Float=&FNTimeOfDayScalar(Night Scale=0.125, Morning Scale=0.2, Day Scale=0.2, Evening Scale=0.15)
      )
     )
    )
    Particles.SpriteSize=&Multiply_Vector2D_ByFloat(
     Float=&FloatFromCurve(
      FloatCurve=DataInterface() {propertyValues {Curve {keys [<at 0 0 RCIM_Constant>,<at 0.0647 1.0636 RCIM_Constant>,<at 0.1658 0 RCIM_Constant>,<at 0.5826 1 RCIM_Constant>,<at 0.7053 0 RCIM_Constant>]}}}
      CurveIndex=&Multiply_Float(A=$Particles.PartitionParticles.Normalized Index In Partition, B=$Emitter.NewScale)
     )
     Vector2D=<25 25>
    )
   ) {isSetParameters=true, module=SetVariables_#23}
   &SkeletalMeshLocation(
    Bone / Socket Index=3
    Bone Sampling Mode=::NewEnumerator4
    MaskSpawnGroup=false
    Orientation Sampling=::NewEnumerator1
    Position Sampling=::NewEnumerator2
    Sampled Position Offset=&Add_Vector(A=<5 0 5>, B=&RandomVector(Vector Scale=3))
    Skeletal Mesh=$User.SkeletalMesh
    Use Sampled Mesh Attribute as Ribbon ID=false
   )
   /SetVariables(Particles.Lerp Position:NiagaraPosition=/Niagara/DynamicInputs/Lerp/Lerp_Position(
     A:NiagaraPosition=$Particles.FN_BoneFlow.Position
     B:NiagaraPosition=Linked() {linkedVariable {name=Particles.SkeletalMeshLocation.SampledPosition, type=Vector3f}}
     Alpha:NiagaraFloat=&FloatFromCurve(
      FloatCurve=DataInterface() {propertyValues {Curve {keys [<at 0 1>,<at 0.01 0>]}}}
      CurveIndex=&Multiply_Float(A=$Particles.PartitionParticles.Normalized Index In Partition, B=$Emitter.NewScale)
     )
    )) {isSetParameters=true}
  ]
  rendererClasses [&NiagaraRibbonRendererProperties, &NiagaraSpriteRendererProperties]
  renderers [
   {rendererClass=&NiagaraRibbonRendererProperties}
   {rendererClass=&NiagaraSpriteRendererProperties, rendererIndex=1}
   {rendererClass=&NiagaraRibbonRendererProperties, rendererIndex=2}
  ]
 }
 Electricity_MainLine003 {
  emitterSpawnScript [
   /SetVariables(Emitter.NumParticles:NiagaraInt32=&MakeCustomIntFromBool(
     Bool=&IsPlatformSetActive(New Platform Set=DataInterface() {propertyValues {Platforms {qualityLevelMask=-31}}})
     True INT=5
     False INT=&MakeCustomIntFromBool(
      Bool=&IsPlatformSetActive(New Platform Set=DataInterface() {propertyValues {Platforms {qualityLevelMask=-30}}})
      True INT=5
      False INT=&MakeCustomIntFromBool(
       Bool=&IsPlatformSetActive(New Platform Set=DataInterface() {propertyValues {Platforms {qualityLevelMask=-28}}})
       True INT=7
       False INT=7
      )
     )
    )) {isSetParameters=true, module=SetVariables_#2}
   /SetVariables(Emitter.NewScale:NiagaraFloat=&Divide_Float(
     A=1
     B=&OneMinusFloat(Float=&Divide_Float(A=1, B=&MakeFloatFromInt(Int=$Emitter.NumParticles)))
    )) {isSetParameters=true, module=SetVariables_#3}
   /SetVariables(Emitter.IsRegularSKin:NiagaraBool=&Boolean_NotOperation(Bool=$System.is alt bool)) {isSetParameters=true}
   /SetVariables(Emitter.FixTrue=true) {isSetParameters=true}
  ]
  emitterUpdateScript [
   /SetVariables(Emitter.Life:NiagaraFloat=&Lerp_Float(
     A=0.75
     B=1
     Alpha=&FloatFromCurve(
      FloatCurve=DataInterface() {propertyValues {Curve {keys [
          {
           arriveTangent=1.2434
           arriveTangentWeight=0.5979
           interpMode=RCIM_Cubic
           leaveTangent=1.2434
           leaveTangentWeight=0.5979
           tangentMode=RCTM_User
           tangentWeightMode=RCTWM_WeightedBoth
           time=0
           value=0
          }
          {
           arriveTangentWeight=0.8985
           interpMode=RCIM_Cubic
           leaveTangentWeight=0.8985
           tangentMode=RCTM_User
           tangentWeightMode=RCTWM_WeightedBoth
           time=1
           value=1
          }
         ]}}}
      CurveIndex=&RandomRangeFloat(Minimum=0, Maximum=1)
     )
    )) {isSetParameters=true, module=SetVariables_#5}
   /SetVariables(Emitter.Life:NiagaraFloat=&Lerp_Float(
     A=0.075
     B=1
     Alpha=&FloatFromCurve(
      FloatCurve=DataInterface() {propertyValues {Curve {keys [
          {
           arriveTangent=1.2434
           arriveTangentWeight=0.5979
           interpMode=RCIM_Cubic
           leaveTangent=1.2434
           leaveTangentWeight=0.5979
           tangentMode=RCTM_User
           tangentWeightMode=RCTWM_WeightedBoth
           time=0
           value=0
          }
          {
           arriveTangentWeight=0.8985
           interpMode=RCIM_Cubic
           leaveTangentWeight=0.8985
           tangentMode=RCTM_User
           tangentWeightMode=RCTWM_WeightedBoth
           time=1
           value=1
          }
         ]}}}
      CurveIndex=&RandomRangeFloat(Minimum=0, Maximum=1)
     )
    )) {enabled=false, isSetParameters=true, module=SetVariables_#6}
   &EmitterState(Life Cycle Mode=::NewEnumerator1, Loop Duration=$Emitter.Life, Recalculate Duration Each Loop=true, UseLoopDelay=false)
   &SpawnBurst_Instantaneous(Spawn Count=$Emitter.NumParticles, Use Spawn Probability=false, UseLoopCountLimit=false)
   &SpawnBurst_Instantaneous(Spawn Count=$Emitter.NumParticles, Spawn Group=1, Use Spawn Probability=false, UseLoopCountLimit=false) {enabled=false, module=SpawnBurst_Instantaneous001}
   /SetVariables(Emitter.RandomRibbonOffset:NiagaraFloat=&RandomRangeFloat(Minimum=0, Maximum=1, Evaluation Type=::NewEnumerator1)) {isSetParameters=true, module=SetVariables_#7}
   /SetVariables(
    Emitter.random boneflow Angle Left:NiagaraFloat=&RandomRangeFloat(Minimum=0, Maximum=0.3333)
    Emitter.random boneflow Angle Right:NiagaraFloat=&RandomRangeFloat(Minimum=0, Maximum=0.3333, Evaluation Type=::NewEnumerator1)
   ) {isSetParameters=true, module=SetVariables_#8}
   /SetVariables(Emitter.Ribbon Life:NiagaraFloat=&RandomRangeFloat(Minimum=1.5, Maximum=2.5, Evaluation Type=::NewEnumerator1)) {isSetParameters=true, module=SetVariables_#9}
   /SetVariables(Emitter.Is Broken ribbon:NiagaraBool=&SetBoolByFloatComparison(A=&RandomRangeFloat(Minimum=0, Maximum=1, Evaluation Type=::NewEnumerator1), B=0.5)) {isSetParameters=true, module=SetVariables_#10}
   /SetVariables(
    Emitter.Random Ribbon Height Left:NiagaraFloat=&RandomRangeFloat(Minimum=0.001, Maximum=0.1, Evaluation Type=::NewEnumerator1)
    Emitter.Random Ribbon Height Right:NiagaraFloat=&RandomRangeFloat(Minimum=0.001, Maximum=0.1, Evaluation Type=::NewEnumerator1)
   ) {isSetParameters=true, module=SetVariables_#11}
   /SetVariables(
    Emitter.Random Ribbon Height Left 2:NiagaraFloat=&RandomRangeFloat(Minimum=9, Maximum=0.9, Evaluation Type=::NewEnumerator1)
    Emitter.Random Ribbon Height Right 2:NiagaraFloat=&RandomRangeFloat(Minimum=0.9, Maximum=0.9, Evaluation Type=::NewEnumerator1)
   ) {isSetParameters=true, module=SetVariables_#12}
   /SetVariables(
    Emitter.UV Offset 1:NiagaraFloat=&RandomRangeFloat(Minimum=-1, Maximum=1, Evaluation Type=::NewEnumerator1)
    Emitter.UV Offset 2:NiagaraFloat=&RandomRangeFloat(Minimum=0.75, Maximum=-0.75, Evaluation Type=::NewEnumerator1)
   ) {isSetParameters=true, module=SetVariables_#13}
   /SetVariables(Emitter.Material Random:NiagaraFloat=&RandomRangeFloat(Minimum=-100, Maximum=100, Evaluation Type=::NewEnumerator1)) {isSetParameters=true, module=SetVariables_#14}
   /SetVariables(Emitter.CutOutOfset:NiagaraFloat=&RandomRangeFloat(Minimum=-0.5, Maximum=0.5, Evaluation Type=::NewEnumerator1)) {isSetParameters=true, module=SetVariables_#15}
   /SetVariables(Emitter.CutOutOfset:NiagaraFloat=&RandomRangeFloat(Minimum=50, Maximum=50, Evaluation Type=::NewEnumerator1)) {isSetParameters=true, module=SetVariables_#16}
   /SetVariables(Emitter.IsShardSkin:NiagaraBool=$System.is alt bool) {isSetParameters=true}
  ]
  particleSpawnScript [
   &InitializeRibbon(
    Color=&MakeCustomLinearColorFromBool(Bool=$System.is alt bool, True Color=<rgba 2 2 8 1>, False Color=<rgba 6.9333 0 8 1>)
    Lifetime=&Add_Float(A=$Emitter.Life, B=1.2)
    Material Random:NiagaraFloat=$Emitter.Material Random
    Ribbon Width=&MakeCustomFloatFromBool(
     Bool=$System.is alt bool
     True Float=&RandomRangeFloat(Minimum=3, Maximum=2.5)
     False Float=&RandomRangeFloat(Minimum=2, Maximum=2.5)
    )
    UsePositionOffset=false
    Write Material Random=true
   )
   &PartitionParticles(
    Particles Per Partition:NiagaraInt32=$Emitter.NumParticles
    Partition Count Mode:ENiagara_InfiniteLoopDuration=::NewEnumerator1
    Partition Mode:ENiagara_PartitionMode=::NewEnumerator1
    Write Partition Index to Ribbon Link Order:ENiagara_AttributeSamplingApplyOutput=::NewEnumerator0
    Write Partition to Ribbon ID=::NewEnumerator0
   )
   /SetVariables(Particles.GetRandomRibbonOffset:NiagaraFloat=$Emitter.RandomRibbonOffset) {isSetParameters=true, module=SetVariables_#17}
   /SetVariables(Particles.Get Random Boneflow Angle:NiagaraFloat=&MakeCustomFloatFromBool(
     Bool=&SetBoolByIntComparison(A=&ModuloInt(A=$Particles.PartitionParticles.Partition, B=2), B=0, Comparison Type=::NewEnumerator0)
     True Float=$Emitter.random boneflow Angle Left
     False Float=$Emitter.random boneflow Angle Right
    )) {isSetParameters=true, module=SetVariables_#18}
   /SetVariables(Particles.Get Is Broken ribbon:NiagaraBool=$Emitter.Is Broken ribbon) {isSetParameters=true, module=SetVariables_#19}
   /SetVariables(Particles.Get Random ribbon height:NiagaraFloat=&MakeCustomFloatFromBool(
     Bool=&SetBoolByIntComparison(A=&ModuloInt(A=$Particles.PartitionParticles.Partition, B=2), B=0, Comparison Type=::NewEnumerator0)
     True Float=$Emitter.Random Ribbon Height Left
     False Float=$Emitter.Random Ribbon Height Right
    )) {isSetParameters=true, module=SetVariables_#20}
   /SetVariables(Particles.Get Uv Offset 1:NiagaraFloat=$Emitter.UV Offset 1, Particles.Get Uv Offset 2:NiagaraFloat=$Emitter.UV Offset 2) {isSetParameters=true, module=SetVariables_#21}
   /SetVariables(Particles.GetcoutoutOffset:NiagaraFloat=$Emitter.CutOutOfset) {isSetParameters=true, module=SetVariables_#22}
  ]
  particleUpdateScript [
   &ParticleState()
   &SampleSkeletalMesh(
    AdditionalYawPitchRoll=false
    Bone / Socket Index=3
    Bone Sampling Mode:ENiagaraBoneSamplingMode=::NewEnumerator4
    Skeletal Mesh:NiagaraDataInterfaceSkeletalMesh=$User.SkeletalMesh
   ) {module=SampleSkeletalMesh002}
   &SampleSkeletalMesh(
    AdditionalYawPitchRoll=false
    Bone / Socket Index=4
    Bone Sampling Mode:ENiagaraBoneSamplingMode=::NewEnumerator4
    Skeletal Mesh:NiagaraDataInterfaceSkeletalMesh=$User.SkeletalMesh
   ) {module=SampleSkeletalMesh001}
   &FN_BoneFlow(
    Angle:NiagaraFloat=&Add_Float(
     A=$Particles.Get Random Boneflow Angle
     B=&Divide_Float(A=&MakeFloatFromInt(Int=$Particles.PartitionParticles.Partition), B=3)
    )
    BoneID=0
    Enable Position Offsets=true
    Height:NiagaraFloat=&Multiply_Float(A=$Particles.PartitionParticles.Normalized Index In Partition, B=$Emitter.NewScale)
    Offset Mid:Vector3f=&VectorFromCurve(
     VectorCurve=DataInterface() {propertyValues {
       XCurve {keys [<at -1 0>,<at 0 -7>,<at 1 0>]}
       YCurve {keys [<at -1 0>, {arriveTangentWeight=0.3333, leaveTangentWeight=0.3333, time=0, value=-5}, <at 1 0>]}
       ZCurve {keys [<at -1 0>,<at 0 0>,<at 1 0>]}
      }}
     CurveIndex=&DotProduct(
      Vector A=&ReturnMeshOrientationAxis(Mesh Orientation Axis=::XAxis, Mesh Orientation Quat=$Particles.SampleSkeletalMesh002.SampledRotation)
      Vector B=&ReturnMeshOrientationAxis(Mesh Orientation Axis=::XAxis, Mesh Orientation Quat=$Particles.SampleSkeletalMesh001.SampledRotation)
     )
    )
    Offset Start:Vector3f=<0 1 0>
    Radius:NiagaraFloat=&FloatFromCurve(
     FloatCurve=DataInterface() {propertyValues {Curve {keys [<at 0 0>,<at 0.1 1>,<at 1 0>]}}}
     CurveIndex=&Multiply_Float(A=$Particles.PartitionParticles.Normalized Index In Partition, B=$Emitter.NewScale)
     Scale Curve=5
    )
    Set Material Dynamic=false
    SkeletalMesh:NiagaraDataInterfaceSkeletalMesh=$User.SkeletalMesh
    Use Sockets Instead=true
   )
   &DynamicMaterialParameters(
    Index 0 Param 1=$Particles.Get Uv Offset 1
    Index 0 Param 2=$Particles.Get Uv Offset 1
    Index 0 Param 3=&FloatFromCurve(
     FloatCurve=DataInterface() {propertyValues {Curve {keys [{tangentMode=RCTM_User, time=0.9, value=0}, {tangentMode=RCTM_User, time=1, value=1}]}}}
     Scale Curve=0
    )
    Index 0 Param 4=&Add_Float(
     A=&Add_Float(
      A=&MakeCustomFloatFromBool(Bool=$System.is alt bool, True Float=0, False Float=0)
      B=&MakeCustomFloatFromBool(
       Bool=$System.is alt bool
       True Float=&FloatFromCurve(
        FloatCurve=DataInterface() {propertyValues {Curve {keys [
            {arriveTangent=0.2, interpMode=RCIM_Cubic, time=0.1, value=-0.005}
            {arriveTangentWeight=0.2567, interpMode=RCIM_Cubic, leaveTangentWeight=0.2567, time=0.5, value=0}
            {
             arriveTangentWeight=0.4239
             interpMode=RCIM_Cubic
             leaveTangentWeight=0.4239
             tangentMode=RCTM_User
             tangentWeightMode=RCTWM_WeightedBoth
             time=1
             value=-0.005
            }
           ]}}}
        CurveIndex=&Multiply_Float(A=$Emitter.NewScale, B=$Particles.PartitionParticles.Normalized Index In Partition)
        Scale Curve=0
       )
       False Float=&FloatFromCurve(
        FloatCurve=DataInterface() {propertyValues {Curve {keys [
            {arriveTangent=0.2, interpMode=RCIM_Cubic, time=0.1, value=-0.025}
            {arriveTangentWeight=0.2567, interpMode=RCIM_Cubic, leaveTangentWeight=0.2567, time=0.5, value=0}
            {
             arriveTangentWeight=0.4239
             interpMode=RCIM_Cubic
             leaveTangentWeight=0.4239
             tangentMode=RCTM_User
             tangentWeightMode=RCTWM_WeightedBoth
             time=1
             value=-0.025
            }
           ]}}}
        CurveIndex=&Multiply_Float(A=$Emitter.NewScale, B=$Particles.PartitionParticles.Normalized Index In Partition)
        Scale Curve=0
       )
      )
     )
     B=&MakeCustomFloatFromBool(
      Bool=$System.is alt bool
      True Float=&RandomRangeFloat(Minimum=0, Maximum=0)
      False Float=&RandomRangeFloat(Minimum=0, Maximum=0)
     )
    )
    Index 1 Param 1=&Multiply_Float(
     A=&FloatFromCurve(
      FloatCurve=DataInterface() {propertyValues {Curve {keys [{tangentMode=RCTM_User, time=0, value=1}, {tangentMode=RCTM_User, time=1, value=1}]}}}
      Scale Curve=&RandomRangeFloat(Minimum=10, Maximum=10)
     )
     B=&FloatFromCurve(
      FloatCurve=DataInterface() {propertyValues {Curve {keys [
          {interpMode=RCIM_Cubic, tangentMode=RCTM_User, time=0, value=0}
          {interpMode=RCIM_Cubic, tangentMode=RCTM_User, time=0.25, value=1}
          {arriveTangentWeight=0.2778, interpMode=RCIM_Cubic, leaveTangentWeight=0.1389, tangentMode=RCTM_User, time=0.75, value=1}
          {tangentMode=RCTM_User, time=1, value=0}
         ]}}}
      CurveIndex=&Multiply_Float(A=$Particles.PartitionParticles.Normalized Index In Partition, B=$Emitter.NewScale)
      Scale Curve=0.66
     )
    )
    Index 2 Param 1=&Add_Float(A=$Particles.GetcoutoutOffset, B=&Multiply_Float(A=$Emitter.LoopedAge, B=-0.025))
    Index 2 Param 2=&MakeCustomFloatFromBool(
     Bool=$System.is alt bool
     True Float=&FN_FrontEndFloatModifier(
      Gameplay=&Lerp_Float(A=0.85, B=1, Alpha=&Sine(Normalized Angle=$Emitter.Age, Period=0.1, Scale=0.5, Bias=0.5))
      Front End=&Lerp_Float(A=1.35, B=1.5, Alpha=&Sine(Normalized Angle=$Emitter.Age, Period=0.1, Scale=0.5, Bias=0.5))
     )
     False Float=&FN_FrontEndFloatModifier(
      Gameplay=&Lerp_Float(A=1.5, B=2, Alpha=&Sine(Normalized Angle=$Emitter.Age, Period=0.1, Scale=0.5, Bias=0.5))
      Front End=&Lerp_Float(A=4, B=5, Alpha=&Sine(Normalized Angle=$Emitter.Age, Period=0.1, Scale=0.5, Bias=0.5))
     )
    )
    Index 2 Param 3=1
    Module.Param0WriteEnabled=true
    Module.Param0WriteEnabled1=true
    Module.Param0WriteEnabled2=true
    Module.Param1WriteEnabled=true
    Module.Param1WriteEnabled1=true
    Module.Param1WriteEnabled2=true
    Module.Param2WriteEnabled=true
    Module.Param2WriteEnabled1=true
    Module.Param2WriteEnabled2=true
    Module.Param3WriteEnabled=true
    Module.Param3WriteEnabled1=true
    Module.Param3WriteEnabled2=true
    Write Parameter Index 1=true
    Write Parameter Index 2=true
   )
   &ShapeLocation(
    Hemisphere Distribution=<1 0.5>
    MaskSpawnGroup=false
    Override Seed=false
    Shape Origin=$Particles.Position
    Sphere Radius=&Multiply_Float(
     A=&FloatFromCurve(FloatCurve=DataInterface() {propertyValues {Curve {keys [{tangentMode=RCTM_User, time=0, value=0.9}, {tangentMode=RCTM_User, time=1, value=1.25}]}}})
     B=&FloatFromCurve(
      FloatCurve=DataInterface() {propertyValues {Curve {keys [
          {interpMode=RCIM_Cubic, tangentMode=RCTM_User, time=0.05, value=0}
          {interpMode=RCIM_Cubic, tangentMode=RCTM_User, time=0.25, value=1}
          {arriveTangentWeight=0.2778, interpMode=RCIM_Cubic, leaveTangentWeight=0.1389, tangentMode=RCTM_User, time=0.7742, value=1}
          {tangentMode=RCTM_User, time=0.95, value=0}
         ]}}}
      CurveIndex=&Multiply_Float(A=$Particles.PartitionParticles.Normalized Index In Partition, B=$Emitter.NewScale)
      Scale Curve=&RandomRangeFloat(Minimum=1, Maximum=1)
     )
    )
    UseOwnerScale=true
   )
   /SetVariables(
    Particles.GlowSpriteColor:LinearColor=&MultiplyLinearColor(
     Linear Color=$Particles.Initial.Color
     Scale Factor=&MakeFloat4FromVectorAndFloat(
      "Vector (RGB)"=<0.3 0.3 0.3>
      "Float (Alpha)"=&FNTimeOfDayScalar(Night Scale=0.125, Morning Scale=0.2, Day Scale=0.2, Evening Scale=0.1)
     )
    )
    Particles.SpriteSize=&Multiply_Vector2D_ByFloat(
     Float=&FloatFromCurve(
      FloatCurve=DataInterface() {propertyValues {Curve {keys [<at 0 0 RCIM_Constant>,<at 0.0647 1.0636 RCIM_Constant>,<at 0.1658 0 RCIM_Constant>,<at 0.5826 1 RCIM_Constant>,<at 0.7053 0 RCIM_Constant>]}}}
      CurveIndex=&Multiply_Float(A=$Particles.PartitionParticles.Normalized Index In Partition, B=$Emitter.NewScale)
     )
     Vector2D=<25 25>
    )
   ) {isSetParameters=true, module=SetVariables_#23}
   /SetVariables(Particles.Position=$Particles.FN_BoneFlow.Position) {isSetParameters=true}
  ]
  rendererClasses [&NiagaraRibbonRendererProperties, &NiagaraSpriteRendererProperties]
  renderers [
   {rendererClass=&NiagaraRibbonRendererProperties}
   {rendererClass=&NiagaraSpriteRendererProperties, rendererIndex=1}
   {rendererClass=&NiagaraRibbonRendererProperties, rendererIndex=2}
  ]
 }
 Electricity_MainLine004 {
  emitterUpdateScript [
   &EmitterState()
   &SpawnBurst_Instantaneous(Age:NiagaraFloat=$Emitter.Age, Spawn Count=3, Use Spawn Probability=false, UseLoopCountLimit=false)
  ]
  particleSpawnScript [&InitializeParticle(
    Color=&MakeCustomLinearColorFromBool(Bool=$System.is alt bool, True Color=<rgba 1.6 1.6 8 1>, False Color=<rgba 6.9333 0 8 1>)
    Sprite Size Mode=::NewEnumerator3
    UsePositionOffset=false
    Write Material Random=true
   )]
  particleUpdateScript [
   &ParticleState(
    Kill Particles When Lifetime Has Elapsed=false
    Let Infinitely Lived Particles Die When Emitter Deactivates=true
    Loop Particles Lifetime=true
   )
   &FN_Light_Attributes(
    Color=&HueShiftLinearColor(
     Hue=0
     Saturation=0.5
     Value=&MakeCustomFloatFromBool(Bool=$System.is alt bool, True Float=0.2, False Float=1)
     Linear Color=$Particles.Initial.Color
    )
    Create Color Attribute=true
    Create Exponent Attribute=true
    Create Radius Attribute=true
    Exponent=3
    Radius=&Waveform("Waveform [1]"=::NewEnumerator10, "[1] Frequency"=10, "[1] Amplitude Min/Max"=<22.5 25>)
    Scale Color=&Waveform("Waveform [1]"=::NewEnumerator10, "[1] Frequency"=10, "[1] Amplitude Min/Max"=<2 2.5>)
   )
   &SkeletalMeshLocation(
    Bone / Socket Index=&SelectIntFromArray(
     Direct Array Index=&ReturnParticleIDAttributes()
     Int Selection Array=DataInterface() {propertyValues {IntData [1,4,3]}}
    )
    Bone / Socket Sampling Mode=::NewEnumerator5
    MaskSpawnGroup=false
    Mesh Sampling Type=::NewEnumerator6
    Sampled Position Offset=&SelectVectorFromArray(
     Direct Array Index=&ReturnParticleIDAttributes()
     Vector Selection Array=DataInterface() {propertyValues {FloatData [<-2 0 2>, <0 0 -1>, <5 -5 0>]}}
    )
    Skeletal Mesh=$User.SkeletalMesh
   )
  ]
  rendererClasses [&NiagaraLightRendererProperties, &NiagaraSpriteRendererProperties]
  renderers [
   {rendererClass=&NiagaraLightRendererProperties, rendererIndex=0}
   {rendererClass=&NiagaraSpriteRendererProperties, rendererIndex=1}
  ]
 }
 Electricity_Short_Bottom {
  emitterSpawnScript [
   /SetVariables(Emitter.NumParticles:NiagaraInt32=&MakeCustomIntFromBool(
     Bool=&IsPlatformSetActive(New Platform Set=DataInterface() {propertyValues {Platforms {qualityLevelMask=-31}}})
     True INT=3
     False INT=&MakeCustomIntFromBool(
      Bool=&IsPlatformSetActive(New Platform Set=DataInterface() {propertyValues {Platforms {qualityLevelMask=-30}}})
      True INT=5
      False INT=&MakeCustomIntFromBool(
       Bool=&IsPlatformSetActive(New Platform Set=DataInterface() {propertyValues {Platforms {qualityLevelMask=-28}}})
       True INT=7
       False INT=3
      )
     )
    )) {isSetParameters=true, module=SetVariables_#2}
   /SetVariables(Emitter.NewScale:NiagaraFloat=&Divide_Float(
     A=1
     B=&OneMinusFloat(Float=&Divide_Float(A=1, B=&MakeFloatFromInt(Int=$Emitter.NumParticles)))
    )) {isSetParameters=true, module=SetVariables_#3}
  ]
  emitterUpdateScript [
   &EmitterState(
    Life Cycle Mode=::NewEnumerator1
    Loop Duration=&RandomRangeFloat(Minimum=0.45, Maximum=0.9)
    Recalculate Duration Each Loop=true
    UseLoopDelay=false
   )
   &SpawnBurst_Instantaneous(Spawn Count=$Emitter.NumParticles, Use Spawn Probability=false, UseLoopCountLimit=false)
   &SpawnBurst_Instantaneous(
    Spawn Count=$Emitter.NumParticles
    Spawn Group=1
    Spawn Time=&RandomRangeFloat(Minimum=0.001, Maximum=0.35)
    Use Spawn Probability=false
    UseLoopCountLimit=false
   ) {module=SpawnBurst_Instantaneous001}
   /SetVariables(Emitter.RandomRibbonOffset:NiagaraFloat=&RandomRangeFloat(Minimum=0, Maximum=1, Evaluation Type=::NewEnumerator1)) {isSetParameters=true, module=SetVariables_#7}
   /SetVariables(
    Emitter.random boneflow Angle Left:NiagaraFloat=&RandomRangeFloat(Minimum=1, Maximum=0)
    Emitter.random boneflow Angle Right:NiagaraFloat=&RandomRangeFloat(Minimum=0, Maximum=1, Evaluation Type=::NewEnumerator1)
   ) {isSetParameters=true, module=SetVariables_#8}
   /SetVariables(Emitter.Ribbon Life:NiagaraFloat=&RandomRangeFloat(Minimum=0.45, Maximum=0.65, Evaluation Type=::NewEnumerator1)) {isSetParameters=true, module=SetVariables_#9}
   /SetVariables(
    Emitter.Random Ribbon Height Left:NiagaraFloat=&RandomRangeFloat(Minimum=0.65, Maximum=0.75, Evaluation Type=::NewEnumerator1)
    Emitter.Random Ribbon Height Right:NiagaraFloat=&RandomRangeFloat(Minimum=0.8, Maximum=0.8, Evaluation Type=::NewEnumerator1)
   ) {isSetParameters=true, module=SetVariables_#11}
   /SetVariables(
    Emitter.Random Ribbon Height Left2:NiagaraFloat=&RandomRangeFloat(Minimum=-0.35, Maximum=-0.5, Evaluation Type=::NewEnumerator1)
    Emitter.Random Ribbon Height Right2:NiagaraFloat=&RandomRangeFloat(Minimum=-0.35, Maximum=-0.5, Evaluation Type=::NewEnumerator1)
   ) {isSetParameters=true}
   /SetVariables(Emitter.ribbobn elec Width:NiagaraFloat=&MakeCustomFloatFromBool(
     Bool=&SetBoolByFloatComparison(A=&RandomRangeFloat(Minimum=0, Maximum=1, Evaluation Type=::NewEnumerator1), B=0.4)
     True Float=&MakeCustomFloatFromBool(Bool=$System.is alt bool, True Float=1.9, False Float=1.4)
     False Float=&MakeCustomFloatFromBool(Bool=$System.is alt bool, True Float=3, False Float=2.4)
    )) {isSetParameters=true}
   /SetVariables(Emitter.BoneFlow Radius:NiagaraFloat=&RandomRangeFloat(Minimum=3, Maximum=6, Evaluation Type=::NewEnumerator1)) {isSetParameters=true}
   /SetVariables(Emitter.Material Random:NiagaraFloat=&RandomRangeFloat(Minimum=-100, Maximum=100, Evaluation Type=::NewEnumerator1)) {isSetParameters=true}
  ]
  particleSpawnScript [
   &InitializeRibbon(
    Color=&MakeCustomLinearColorFromBool(
     Bool=$System.is alt bool
     True Color=&FN_Select_Color_from_Time_Of_Day(
      Morning=<rgba 0.25 0.25 1 1>
      Day=&LerpLinearColors(
       StartColor=<rgba 0.25 0.25 1 1>
       EndColor=<rgba 0.3375 0.3375 1.35 1>
       LerpFactor=&FN_FrontEndFloatModifier(Gameplay=0, Front End=1)
      )
      Evening=<rgba 0.25 0.25 1 1>
      Night=<rgba 0.375 0.375 1.5 1>
     )
     False Color=<rgba 1.33 0.225 1.5 1>
    )
    Lifetime=$Emitter.Ribbon Life
    Material Random:NiagaraFloat=$Emitter.Material Random
    Ribbon Width=$Emitter.ribbobn elec Width
    UsePositionOffset=false
    Write Material Random=true
   )
   &PartitionParticles(
    Particles Per Partition:NiagaraInt32=$Emitter.NumParticles
    Partition Count Mode:ENiagara_InfiniteLoopDuration=::NewEnumerator1
    Partition Mode:ENiagara_PartitionMode=::NewEnumerator1
    Write Partition Index to Ribbon Link Order:ENiagara_AttributeSamplingApplyOutput=::NewEnumerator0
    Write Partition to Ribbon ID=::NewEnumerator0
   )
   /SetVariables(Particles.GetRandomRibbonOffset:NiagaraFloat=$Emitter.RandomRibbonOffset) {isSetParameters=true, module=SetVariables_#17}
   /SetVariables(Particles.Get Random Boneflow Angle:NiagaraFloat=&MakeCustomFloatFromBool(
     Bool=&SetBoolByIntComparison(A=&ModuloInt(A=$Particles.PartitionParticles.Partition, B=2), B=0, Comparison Type=::NewEnumerator0)
     True Float=$Emitter.random boneflow Angle Left
     False Float=$Emitter.random boneflow Angle Right
    )) {isSetParameters=true, module=SetVariables_#18}
   /SetVariables(
    Particles.Get Random ribbon height:NiagaraFloat=&MakeCustomFloatFromBool(
     Bool=&SetBoolByIntComparison(A=&ModuloInt(A=$Particles.PartitionParticles.Partition, B=2), B=0, Comparison Type=::NewEnumerator0)
     True Float=$Emitter.Random Ribbon Height Left
     False Float=$Emitter.Random Ribbon Height Right
    )
    Particles.Get Random ribbon height 2:NiagaraFloat=&MakeCustomFloatFromBool(
     Bool=&SetBoolByIntComparison(A=&ModuloInt(A=$Particles.PartitionParticles.Partition, B=2), B=0, Comparison Type=::NewEnumerator0)
     True Float=$Emitter.Random Ribbon Height Left2
     False Float=$Emitter.Random Ribbon Height Right2
    )
   ) {isSetParameters=true, module=SetVariables_#20}
   /SetVariables(Particles.Get bone Flow Radius:NiagaraFloat=$Emitter.BoneFlow Radius) {isSetParameters=true, module=SetVariables_#24}
  ]
  particleUpdateScript [
   &ParticleState()
   &SampleSkeletalMesh(
    AdditionalYawPitchRoll=false
    Bone Sampling Mode:ENiagaraBoneSamplingMode=::NewEnumerator4
    Skeletal Mesh:NiagaraDataInterfaceSkeletalMesh=$User.SkeletalMesh
   )
   &SampleSkeletalMesh(
    AdditionalYawPitchRoll=false
    Bone / Socket Index=1
    Bone Sampling Mode:ENiagaraBoneSamplingMode=::NewEnumerator4
    Skeletal Mesh:NiagaraDataInterfaceSkeletalMesh=$User.SkeletalMesh
   ) {module=SampleSkeletalMesh003}
   &SampleSkeletalMesh(
    AdditionalYawPitchRoll=false
    Bone / Socket Index=3
    Bone Sampling Mode:ENiagaraBoneSamplingMode=::NewEnumerator4
    Skeletal Mesh:NiagaraDataInterfaceSkeletalMesh=$User.SkeletalMesh
   ) {module=SampleSkeletalMesh002}
   &SampleSkeletalMesh(
    AdditionalYawPitchRoll=false
    Bone / Socket Index=4
    Bone Sampling Mode:ENiagaraBoneSamplingMode=::NewEnumerator4
    Skeletal Mesh:NiagaraDataInterfaceSkeletalMesh=$User.SkeletalMesh
   ) {module=SampleSkeletalMesh001}
   &FN_BoneFlow(
    Angle:NiagaraFloat=&Add_Float(
     A=&Add_Float(
      A=$Particles.Get Random Boneflow Angle
      B=&FloatFromCurve(
       FloatCurve=DataInterface() {propertyValues {Curve {keys [<at 0 0.25>,<at 0.1 0.1>,<at 0.9 -0.1>,<at 1 -0.25>]}}}
       CurveIndex=&RandomRangeFloat(Minimum=0, Maximum=1)
       Scale Curve=0
      )
     )
     B=&FloatFromCurve(
      FloatCurve=DataInterface() {propertyValues {Curve {keys [
          {interpMode=RCIM_Cubic, tangentMode=RCTM_User, time=0, value=0}
          {interpMode=RCIM_Cubic, tangentMode=RCTM_User, time=0.5, value=1}
          {tangentMode=RCTM_User, time=1, value=0}
         ]}}}
      CurveIndex=&Multiply_Float(A=$Emitter.NewScale, B=$Particles.PartitionParticles.Normalized Index In Partition)
     )
    )
    BoneID:NiagaraInt32=&ModuloInt(A=$Particles.PartitionParticles.Partition, B=2)
    Height:NiagaraFloat=&Lerp_Float(
     A=$Particles.Get Random ribbon height
     B=&Add_Float(A=$Particles.Get Random ribbon height, B=$Particles.Get Random ribbon height 2)
     Alpha=&Multiply_Float(A=$Particles.PartitionParticles.Normalized Index In Partition, B=$Emitter.NewScale)
    )
    Radius:NiagaraFloat=&Multiply_Float(
     A=&Multiply_Float(
      A=$Particles.Initial.Get bone Flow Radius
      B=&FloatFromCurve(
       FloatCurve=DataInterface() {propertyValues {Curve {keys [<at 0 1 RCIM_Cubic>, {arriveTangent=-5.5359, interpMode=RCIM_Cubic, leaveTangent=-5.5359, tangentMode=RCTM_User, time=1, value=0}]}}}
       CurveIndex=&Lerp_Float(
        A=$Particles.Get Random ribbon height
        B=&Add_Float(A=$Particles.Get Random ribbon height, B=$Particles.Get Random ribbon height 2)
        Alpha=&Multiply_Float(A=$Particles.PartitionParticles.Normalized Index In Partition, B=$Emitter.NewScale)
       )
      )
     )
     B=&FloatFromCurve(
      FloatCurve=DataInterface() {propertyValues {Curve {keys [{tangentMode=RCTM_User, time=0.2, value=0.75}, {tangentMode=RCTM_User, time=0.4, value=1}]}}}
      CurveIndex=&RandomRangeFloat(Minimum=0, Maximum=1)
      Scale Curve=0.5
     )
    )
    SkeletalMesh:NiagaraDataInterfaceSkeletalMesh=$User.SkeletalMesh
   )
   &ScaleColor(
    Scale Alpha=2
    Scale RGB=&Multiply_VectorByFloat(
     Vector=&VectorFromFloat(Value=&MakeCustomFloatFromBool(
       Bool=$System.is alt bool
       True Float=&FN_FrontEndFloatModifier(
        Gameplay=&Lerp_Float(A=2.2, B=1.75, Alpha=&Sine(Normalized Angle=$Emitter.Age, Period=0.1, Scale=0.5, Bias=0.5))
        Front End=&Lerp_Float(A=1.75, B=2, Alpha=&Sine(Normalized Angle=$Emitter.Age, Period=0.1, Scale=0.5, Bias=0.5))
       )
       False Float=&FN_FrontEndFloatModifier(
        Gameplay=&Lerp_Float(A=1.25, B=1, Alpha=&Sine(Normalized Angle=$Emitter.Age, Period=0.1, Scale=0.5, Bias=0.5))
        Front End=&Lerp_Float(A=3, B=2.5, Alpha=&Sine(Normalized Angle=$Emitter.Age, Period=0.1, Scale=0.5, Bias=0.5))
       )
      ))
     Float=&MakeCustomFloatFromBool(
      Bool=$System.is alt bool
      True Float=&FNTimeOfDayScalar(Night Scale=0.2, Morning Scale=0.1, Day Scale=0.5, Evening Scale=0.15)
      False Float=&FNTimeOfDayScalar(Night Scale=0.65, Morning Scale=0.5, Day Scale=0.5, Evening Scale=0.5)
     )
    )
    ScaleA=true
    ScaleRGB=true
   ) {module=ScaleColor001}
   &ScaleRibbonWidth(Ribbon Width Scale=&FloatFromCurve(FloatCurve=DataInterface() {propertyValues {Curve {keys [
         {arriveTangent=6.944, interpMode=RCIM_Cubic, leaveTangent=6.944, tangentMode=RCTM_User, time=0, value=0}
         {
          arriveTangent=0.114
          arriveTangentWeight=0.4848
          interpMode=RCIM_Cubic
          leaveTangentWeight=0.3804
          tangentMode=RCTM_Break
          tangentWeightMode=RCTWM_WeightedBoth
          time=0.5
          value=1
         }
         <at 1 1.5>
        ]}}}))
   &FN_Light_Attributes(
    Color=&HueShiftLinearColor(Hue=0, Saturation=0.5, Value=1, Linear Color=$Particles.Initial.Color)
    Create Color Attribute=true
    Create Exponent Attribute=true
    Create Radius Attribute=true
    Enable Particle Light Enable=true
    Exponent=3
    Light Enabled:NiagaraBool=&Boolean_AndOperation(
     Bool1=true
     Bool2=&SetBoolByFloatComparison(
      A=&FloatFromCurve(
       FloatCurve=DataInterface() {propertyValues {Curve {keys [<at 0 1>,<at 0.01 0>,<at 0.49 0>,<at 0.5 1>,<at 0.51 0>]}}}
       CurveIndex=&Multiply_Float(A=$Particles.PartitionParticles.Normalized Index In Partition, B=$Emitter.NewScale)
      )
      B=0.5
     )
    )
    Radius=&FloatFromCurve(
     FloatCurve=DataInterface() {propertyValues {Curve {keys [
         {interpMode=RCIM_Cubic, tangentMode=RCTM_User, time=0, value=0.5}
         {interpMode=RCIM_Cubic, tangentMode=RCTM_User, time=0.5, value=1}
         {tangentMode=RCTM_User, time=1, value=0.5}
        ]}}}
     Scale Curve=20
    )
    Scale Color=&FloatFromCurve(
     FloatCurve=DataInterface() {propertyValues {Curve {keys [
         {interpMode=RCIM_Cubic, tangentMode=RCTM_User, time=0, value=0.75}
         {interpMode=RCIM_Cubic, tangentMode=RCTM_User, time=0.5, value=1}
         <at 1 0.75 RCIM_Cubic>
        ]}}}
     Scale Curve=2
    )
   )
   &DynamicMaterialParameters(
    Index 0 Param 3=&FloatFromCurve(FloatCurve=DataInterface() {propertyValues {Curve {keys [{tangentMode=RCTM_User, time=0.75, value=0}, {tangentMode=RCTM_User, time=1, value=1}]}}})
    Index 1 Param 1=&FloatFromCurve(
     FloatCurve=DataInterface() {propertyValues {Curve {keys [{tangentMode=RCTM_User, time=0.25, value=0}, {tangentMode=RCTM_User, time=1, value=1}]}}}
     Scale Curve=&RandomRangeFloat(Minimum=0, Maximum=3)
    )
    Module.Param0WriteEnabled=true
    Module.Param0WriteEnabled1=true
    Module.Param1WriteEnabled=true
    Module.Param1WriteEnabled1=true
    Module.Param2WriteEnabled=true
    Module.Param2WriteEnabled1=true
    Module.Param3WriteEnabled=true
    Module.Param3WriteEnabled1=true
    Write Parameter Index 1=true
   ) {module=DynamicMaterialParameters001}
   &ShapeLocation(
    MaskSpawnGroup=false
    Override Seed=false
    Shape Origin=$Particles.Position
    Sphere Radius=&RandomRangeFloat(Minimum=5, Maximum=7)
    UseOwnerScale=true
   )
   /SetVariables(Particles.Position=&Add_Vector(A=$Particles.ShapeLocation.ShapePosition, B=$Particles.FN_BoneFlow.Position)) {isSetParameters=true, module=SetVariables_#4}
  ]
  rendererClasses [
   &NiagaraLightRendererProperties
   &NiagaraRibbonRendererProperties
   &NiagaraSpriteRendererProperties
  ]
  renderers [
   {rendererClass=&NiagaraLightRendererProperties, rendererIndex=0}
   {rendererClass=&NiagaraRibbonRendererProperties, rendererIndex=1}
   {rendererClass=&NiagaraSpriteRendererProperties, rendererIndex=2}
  ]
 }
 Electricity_Short_Tops {
  emitterSpawnScript [
   /SetVariables(Emitter.NumParticles:NiagaraInt32=&MakeCustomIntFromBool(
     Bool=&IsPlatformSetActive(New Platform Set=DataInterface() {propertyValues {Platforms {qualityLevelMask=-31}}})
     True INT=3
     False INT=&MakeCustomIntFromBool(
      Bool=&IsPlatformSetActive(New Platform Set=DataInterface() {propertyValues {Platforms {qualityLevelMask=-30}}})
      True INT=5
      False INT=&MakeCustomIntFromBool(
       Bool=&IsPlatformSetActive(New Platform Set=DataInterface() {propertyValues {Platforms {qualityLevelMask=-28}}})
       True INT=7
       False INT=3
      )
     )
    )) {isSetParameters=true, module=SetVariables_#2}
   /SetVariables(Emitter.NewScale:NiagaraFloat=&Divide_Float(
     A=1
     B=&OneMinusFloat(Float=&Divide_Float(A=1, B=&MakeFloatFromInt(Int=$Emitter.NumParticles)))
    )) {isSetParameters=true, module=SetVariables_#3}
  ]
  emitterUpdateScript [
   &EmitterState(
    Life Cycle Mode=::NewEnumerator1
    Loop Duration=&RandomRangeFloat(Minimum=0.45, Maximum=0.9)
    Recalculate Duration Each Loop=true
    UseLoopDelay=false
   )
   &SpawnBurst_Instantaneous(Spawn Count=$Emitter.NumParticles, Use Spawn Probability=false, UseLoopCountLimit=false)
   &SpawnBurst_Instantaneous(
    Spawn Count=$Emitter.NumParticles
    Spawn Group=1
    Spawn Time=&RandomRangeFloat(Minimum=0.001, Maximum=0.35)
    Use Spawn Probability=false
    UseLoopCountLimit=false
   ) {module=SpawnBurst_Instantaneous001}
   /SetVariables(Emitter.RandomRibbonOffset:NiagaraFloat=&RandomRangeFloat(Minimum=0, Maximum=1, Evaluation Type=::NewEnumerator1)) {isSetParameters=true}
   /SetVariables(
    Emitter.random boneflow Angle Left:NiagaraFloat=&RandomRangeFloat(Minimum=-0.2, Maximum=0.6)
    Emitter.random boneflow Angle Right:NiagaraFloat=&RandomRangeFloat(Minimum=-0.2, Maximum=0.4, Evaluation Type=::NewEnumerator1)
   ) {isSetParameters=true}
   /SetVariables(Emitter.Ribbon Life:NiagaraFloat=&RandomRangeFloat(Minimum=0.45, Maximum=0.65, Evaluation Type=::NewEnumerator1)) {isSetParameters=true}
   /SetVariables(
    Emitter.Random Ribbon Height Left:NiagaraFloat=&RandomRangeFloat(Minimum=0.75, Maximum=0.25, Evaluation Type=::NewEnumerator1)
    Emitter.Random Ribbon Height Right:NiagaraFloat=&RandomRangeFloat(Minimum=0.25, Maximum=1, Evaluation Type=::NewEnumerator1)
   ) {isSetParameters=true}
   /SetVariables(
    Emitter.Random Ribbon Height Left2:NiagaraFloat=&RandomRangeFloat(Minimum=-0.5, Maximum=-0.35, Evaluation Type=::NewEnumerator1)
    Emitter.Random Ribbon Height Right2:NiagaraFloat=&RandomRangeFloat(Minimum=-0.5, Maximum=-0.35, Evaluation Type=::NewEnumerator1)
   ) {isSetParameters=true}
   /SetVariables(Emitter.ribbobn elec Width:NiagaraFloat=&MakeCustomFloatFromBool(
     Bool=&SetBoolByFloatComparison(A=&RandomRangeFloat(Minimum=0, Maximum=1, Evaluation Type=::NewEnumerator1), B=0.4)
     True Float=&MakeCustomFloatFromBool(Bool=$System.is alt bool, True Float=1.9, False Float=1.4)
     False Float=&MakeCustomFloatFromBool(Bool=$System.is alt bool, True Float=3, False Float=2.4)
    )) {isSetParameters=true}
   /SetVariables(Emitter.BoneFlow Radius:NiagaraFloat=&RandomRangeFloat(Minimum=3, Maximum=6, Evaluation Type=::NewEnumerator1)) {isSetParameters=true}
   /SetVariables(Emitter.Material Random:NiagaraFloat=&RandomRangeFloat(Minimum=-100, Maximum=100, Evaluation Type=::NewEnumerator1)) {isSetParameters=true}
  ]
  particleSpawnScript [
   &InitializeRibbon(
    Color=&MakeCustomLinearColorFromBool(
     Bool=$System.is alt bool
     True Color=&FN_Select_Color_from_Time_Of_Day(
      Morning=<rgba 0.25 0.25 1 1>
      Day=&LerpLinearColors(
       StartColor=<rgba 0.25 0.25 1 1>
       EndColor=<rgba 0.3375 0.3375 1.35 1>
       LerpFactor=&FN_FrontEndFloatModifier(Gameplay=0, Front End=1)
      )
      Evening=<rgba 0.25 0.25 1 1>
      Night=<rgba 0.375 0.375 1.5 1>
     )
     False Color=<rgba 1.33 0.225 1.5 1>
    )
    Lifetime=$Emitter.Ribbon Life
    Material Random:NiagaraFloat=$Emitter.Material Random
    Ribbon Width=$Emitter.ribbobn elec Width
    UsePositionOffset=false
    Write Material Random=true
   )
   &PartitionParticles(
    Particles Per Partition:NiagaraInt32=$Emitter.NumParticles
    Partition Count Mode:ENiagara_InfiniteLoopDuration=::NewEnumerator1
    Partition Mode:ENiagara_PartitionMode=::NewEnumerator1
    Write Partition Index to Ribbon Link Order:ENiagara_AttributeSamplingApplyOutput=::NewEnumerator0
    Write Partition to Ribbon ID=::NewEnumerator0
   )
   /SetVariables(Particles.GetRandomRibbonOffset:NiagaraFloat=$Emitter.RandomRibbonOffset) {isSetParameters=true, module=SetVariables_#17}
   /SetVariables(Particles.Get Random Boneflow Angle:NiagaraFloat=&MakeCustomFloatFromBool(
     Bool=&SetBoolByIntComparison(A=&ModuloInt(A=$Particles.PartitionParticles.Partition, B=2), B=0, Comparison Type=::NewEnumerator0)
     True Float=$Emitter.random boneflow Angle Left
     False Float=$Emitter.random boneflow Angle Right
    )) {isSetParameters=true, module=SetVariables_#18}
   /SetVariables(
    Particles.Get Random ribbon height:NiagaraFloat=&MakeCustomFloatFromBool(
     Bool=&SetBoolByIntComparison(A=&ModuloInt(A=$Particles.PartitionParticles.Partition, B=2), B=0, Comparison Type=::NewEnumerator0)
     True Float=$Emitter.Random Ribbon Height Left
     False Float=$Emitter.Random Ribbon Height Right
    )
    Particles.Get Random ribbon height 2:NiagaraFloat=&MakeCustomFloatFromBool(
     Bool=&SetBoolByIntComparison(A=&ModuloInt(A=$Particles.PartitionParticles.Partition, B=2), B=0, Comparison Type=::NewEnumerator0)
     True Float=$Emitter.Random Ribbon Height Left2
     False Float=$Emitter.Random Ribbon Height Right2
    )
   ) {isSetParameters=true, module=SetVariables_#20}
   /SetVariables(Particles.Get bone Flow Radius:NiagaraFloat=$Emitter.BoneFlow Radius) {isSetParameters=true, module=SetVariables_#24}
  ]
  particleUpdateScript [
   &ParticleState()
   &SampleSkeletalMesh(
    AdditionalYawPitchRoll=false
    Bone Sampling Mode:ENiagaraBoneSamplingMode=::NewEnumerator4
    Skeletal Mesh:NiagaraDataInterfaceSkeletalMesh=$User.SkeletalMesh
   )
   &SampleSkeletalMesh(
    AdditionalYawPitchRoll=false
    Bone / Socket Index=1
    Bone Sampling Mode:ENiagaraBoneSamplingMode=::NewEnumerator4
    Skeletal Mesh:NiagaraDataInterfaceSkeletalMesh=$User.SkeletalMesh
   ) {module=SampleSkeletalMesh003}
   &SampleSkeletalMesh(
    AdditionalYawPitchRoll=false
    Bone / Socket Index=3
    Bone Sampling Mode:ENiagaraBoneSamplingMode=::NewEnumerator4
    Skeletal Mesh:NiagaraDataInterfaceSkeletalMesh=$User.SkeletalMesh
   ) {module=SampleSkeletalMesh002}
   &SampleSkeletalMesh(
    AdditionalYawPitchRoll=false
    Bone / Socket Index=4
    Bone Sampling Mode:ENiagaraBoneSamplingMode=::NewEnumerator4
    Skeletal Mesh:NiagaraDataInterfaceSkeletalMesh=$User.SkeletalMesh
   ) {module=SampleSkeletalMesh001}
   &FN_BoneFlow(
    Angle:NiagaraFloat=&Add_Float(
     A=&Add_Float(
      A=$Particles.Get Random Boneflow Angle
      B=&FloatFromCurve(
       FloatCurve=DataInterface() {propertyValues {Curve {keys [<at 0 0.25>,<at 0.1 0.1>,<at 0.9 -0.1>,<at 1 -0.25>]}}}
       CurveIndex=&RandomRangeFloat(Minimum=0, Maximum=1)
       Scale Curve=0
      )
     )
     B=&FloatFromCurve(
      FloatCurve=DataInterface() {propertyValues {Curve {keys [
          {interpMode=RCIM_Cubic, tangentMode=RCTM_User, time=0, value=0}
          {interpMode=RCIM_Cubic, tangentMode=RCTM_User, time=0.5, value=1}
          {tangentMode=RCTM_User, time=1, value=0}
         ]}}}
      CurveIndex=&Multiply_Float(A=$Emitter.NewScale, B=$Particles.PartitionParticles.Normalized Index In Partition)
     )
    )
    BoneID:NiagaraInt32=&ModuloInt(A=$Particles.PartitionParticles.Partition, B=2)
    Height:NiagaraFloat=&Lerp_Float(
     A=$Particles.Get Random ribbon height
     B=&Add_Float(A=$Particles.Get Random ribbon height, B=$Particles.Get Random ribbon height 2)
     Alpha=&Multiply_Float(A=$Particles.PartitionParticles.Normalized Index In Partition, B=$Emitter.NewScale)
    )
    Radius:NiagaraFloat=&Multiply_Float(
     A=&Multiply_Float(
      A=$Particles.Initial.Get bone Flow Radius
      B=&FloatFromCurve(
       FloatCurve=DataInterface() {propertyValues {Curve {keys [<at 0 1 RCIM_Cubic>, {arriveTangent=-5.5359, interpMode=RCIM_Cubic, leaveTangent=-5.5359, tangentMode=RCTM_User, time=1, value=0}]}}}
       CurveIndex=&Lerp_Float(
        A=$Particles.Get Random ribbon height
        B=&Add_Float(A=$Particles.Get Random ribbon height, B=$Particles.Get Random ribbon height 2)
        Alpha=&Multiply_Float(A=$Particles.PartitionParticles.Normalized Index In Partition, B=$Emitter.NewScale)
       )
      )
     )
     B=&FloatFromCurve(
      FloatCurve=DataInterface() {propertyValues {Curve {keys [{tangentMode=RCTM_User, time=0.2, value=0.75}, {tangentMode=RCTM_User, time=0.4, value=1}]}}}
      CurveIndex=&RandomRangeFloat(Minimum=0, Maximum=1)
      Scale Curve=0.5
     )
    )
    Set Material Dynamic=false
    SkeletalMesh:NiagaraDataInterfaceSkeletalMesh=$User.SkeletalMesh
   )
   &ScaleColor(
    Scale Alpha=2
    Scale RGB=&Multiply_VectorByFloat(
     Vector=&VectorFromFloat(Value=&MakeCustomFloatFromBool(
       Bool=$System.is alt bool
       True Float=&FN_FrontEndFloatModifier(
        Gameplay=&Lerp_Float(A=2.2, B=1.75, Alpha=&Sine(Normalized Angle=$Emitter.Age, Period=0.1, Scale=0.5, Bias=0.5))
        Front End=&Lerp_Float(A=1.75, B=2, Alpha=&Sine(Normalized Angle=$Emitter.Age, Period=0.1, Scale=0.5, Bias=0.5))
       )
       False Float=&FN_FrontEndFloatModifier(
        Gameplay=&Lerp_Float(A=1.25, B=1, Alpha=&Sine(Normalized Angle=$Emitter.Age, Period=0.1, Scale=0.5, Bias=0.5))
        Front End=&Lerp_Float(A=3, B=2.5, Alpha=&Sine(Normalized Angle=$Emitter.Age, Period=0.1, Scale=0.5, Bias=0.5))
       )
      ))
     Float=&MakeCustomFloatFromBool(
      Bool=$System.is alt bool
      True Float=&FNTimeOfDayScalar(Night Scale=0.2, Morning Scale=0.1, Day Scale=0.5, Evening Scale=0.15)
      False Float=&FNTimeOfDayScalar(Night Scale=0.65, Morning Scale=0.5, Day Scale=0.5, Evening Scale=0.5)
     )
    )
    ScaleA=true
    ScaleRGB=true
   ) {module=ScaleColor001}
   &ScaleRibbonWidth(Ribbon Width Scale=&FloatFromCurve(FloatCurve=DataInterface() {propertyValues {Curve {keys [
         {arriveTangent=6.944, interpMode=RCIM_Cubic, leaveTangent=6.944, tangentMode=RCTM_User, time=0, value=0}
         {
          arriveTangent=0.114
          arriveTangentWeight=0.4848
          interpMode=RCIM_Cubic
          leaveTangentWeight=0.3804
          tangentMode=RCTM_Break
          tangentWeightMode=RCTWM_WeightedBoth
          time=0.5
          value=1
         }
         <at 1 1.5>
        ]}}}))
   &FN_Light_Attributes(
    Color=&HueShiftLinearColor(Hue=0, Saturation=0.5, Value=1, Linear Color=$Particles.Initial.Color)
    Create Color Attribute=true
    Create Exponent Attribute=true
    Create Radius Attribute=true
    Enable Particle Light Enable=true
    Exponent=3
    Light Enabled:NiagaraBool=&Boolean_AndOperation(
     Bool1=true
     Bool2=&SetBoolByFloatComparison(
      A=&FloatFromCurve(
       FloatCurve=DataInterface() {propertyValues {Curve {keys [<at 0 1>,<at 0.01 0>,<at 0.49 0>,<at 0.5 1>,<at 0.51 0>]}}}
       CurveIndex=&Multiply_Float(A=$Particles.PartitionParticles.Normalized Index In Partition, B=$Emitter.NewScale)
      )
      B=0.5
     )
    )
    Radius=&FloatFromCurve(
     FloatCurve=DataInterface() {propertyValues {Curve {keys [
         {interpMode=RCIM_Cubic, tangentMode=RCTM_User, time=0, value=0.5}
         {interpMode=RCIM_Cubic, tangentMode=RCTM_User, time=0.5, value=1}
         {tangentMode=RCTM_User, time=1, value=0.5}
        ]}}}
     Scale Curve=20
    )
    Scale Color=&FloatFromCurve(
     FloatCurve=DataInterface() {propertyValues {Curve {keys [
         {interpMode=RCIM_Cubic, tangentMode=RCTM_User, time=0, value=0.75}
         {interpMode=RCIM_Cubic, tangentMode=RCTM_User, time=0.5, value=1}
         <at 1 0.75 RCIM_Cubic>
        ]}}}
     Scale Curve=2
    )
   )
   &DynamicMaterialParameters(
    Index 0 Param 3=&FloatFromCurve(FloatCurve=DataInterface() {propertyValues {Curve {keys [{tangentMode=RCTM_User, time=0.75, value=0}, {tangentMode=RCTM_User, time=1, value=1}]}}})
    Index 1 Param 1=&FloatFromCurve(
     FloatCurve=DataInterface() {propertyValues {Curve {keys [{tangentMode=RCTM_User, time=0.25, value=0}, {tangentMode=RCTM_User, time=1, value=1}]}}}
     Scale Curve=&RandomRangeFloat(Minimum=0, Maximum=3)
    )
    Module.Param0WriteEnabled=true
    Module.Param0WriteEnabled1=true
    Module.Param1WriteEnabled=true
    Module.Param1WriteEnabled1=true
    Module.Param2WriteEnabled=true
    Module.Param2WriteEnabled1=true
    Module.Param3WriteEnabled=true
    Module.Param3WriteEnabled1=true
    Write Parameter Index 1=true
   ) {module=DynamicMaterialParameters001}
   &ShapeLocation(
    MaskSpawnGroup=false
    Override Seed=false
    Shape Origin=$Particles.Position
    Sphere Radius=&RandomRangeFloat(Minimum=5, Maximum=7)
    UseOwnerScale=true
   )
   /SetVariables(Particles.Position=&Add_Vector(A=$Particles.ShapeLocation.ShapePosition, B=$Particles.FN_BoneFlow.Position)) {isSetParameters=true, module=SetVariables_#4}
  ]
  rendererClasses [
   &NiagaraLightRendererProperties
   &NiagaraRibbonRendererProperties
   &NiagaraSpriteRendererProperties
  ]
  renderers [
   {rendererClass=&NiagaraLightRendererProperties, rendererIndex=0}
   {rendererClass=&NiagaraRibbonRendererProperties, rendererIndex=1}
   {rendererClass=&NiagaraSpriteRendererProperties, rendererIndex=2}
  ]
 }
}
summary {
 systemName=NS_Forsake_Character
 emitters [
  {
   emitterName=Electricity_Short_Bottom
   rendererClasses [
    &NiagaraLightRendererProperties
    &NiagaraRibbonRendererProperties
    &NiagaraSpriteRendererProperties
   ]
  }
  {
   emitterName=Electricity_MainLine
   rendererClasses [&NiagaraRibbonRendererProperties, &NiagaraSpriteRendererProperties]
  }
  {emitterName=CakeSprinkles, rendererClasses [&NiagaraSpriteRendererProperties]}
  {
   emitterName=Electricity_MainLine003
   rendererClasses [&NiagaraRibbonRendererProperties, &NiagaraSpriteRendererProperties]
  }
  {
   emitterName=Electricity_MainLine004
   rendererClasses [&NiagaraLightRendererProperties, &NiagaraSpriteRendererProperties]
  }
  {
   emitterName=Electricity_Short_Tops
   rendererClasses [
    &NiagaraLightRendererProperties
    &NiagaraRibbonRendererProperties
    &NiagaraSpriteRendererProperties
   ]
  }
 ]
 userVariables [
  User.bIsFrontEnd {type=NiagaraBool}
  User.bIsFrontEndPreview {type=NiagaraBool}
  User.IsAlt {type=NiagaraFloat, defaultValue=0, description=""}
  User.SkeletalMesh {
   type=NiagaraDataInterfaceSkeletalMesh
   defaultValue {
    struct=NiagaraExt_VariableValue_DataInterface
    value {dataInterface=&self:NiagaraDataInterfaceSkeletalMesh_0}
   }
  }
 ]
}
