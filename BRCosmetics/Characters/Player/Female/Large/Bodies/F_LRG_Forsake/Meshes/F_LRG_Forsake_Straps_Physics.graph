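# Physics asset for the side straps on the F_LRG_Forsake large female player body cosmetic, defining their ragdoll-style bodies and joints.
# Each side has a chain of four box bodies hanging from a small buckle-parent sphere, linked by identical constraints with moderate swing and tight twist limits.
# Pelvis and thigh capsules, plus buckle-base spheres, have no constraints; the capsules provide body shapes for the straps to collide against.
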PhysicsAsset /BRCosmetics/Characters/Player/Female/Large/Bodies/F_LRG_Forsake/Meshes/F_LRG_Forsake_Straps_Physics
bodies {
 dyn_side_buckle_base_l [{shapeType=Sphere, radius=0.5, shapeName=dyn_side_buckle_base_l_sphere}]
 dyn_side_buckle_base_r [{shapeType=Sphere, radius=0.5, shapeName=dyn_side_buckle_base_r_sphere}]
 dyn_side_buckle_parent_l [{shapeType=Sphere, radius=0.5, shapeName=dyn_side_buckle_parent_l_sphere}]
 dyn_side_buckle_parent_r [{shapeType=Sphere, radius=0.5, shapeName=dyn_side_buckle_parent_r_sphere}]
 dyn_side_strap_01_l [{shapeType=Box, extentX=4.4508, extentY=6.0413, extentZ=1.0839, shapeName=dyn_side_strap_01_l_box, center=<2.1966 0 0>}]
 dyn_side_strap_01_r [{shapeType=Box, extentX=4.4179, extentY=7.5, extentZ=0.9368, shapeName=dyn_side_strap_01_r_box, center=<2.3114 0 0>}]
 dyn_side_strap_02_l [{shapeType=Box, extentX=5.7142, extentY=5.9615, extentZ=1.9405, shapeName=dyn_side_strap_02_l_box, center=<2.7761 0 0.2258>}]
 dyn_side_strap_02_r [{
   shapeType=Box
   extentX=5.1419
   extentY=7.5
   extentZ=2.0634
   shapeName=dyn_side_strap_02_r_box
   center=<2.9475 -0.0016 0.0497>
   rotation=<rot 0 0.0001 -1.8207>
  }]
 dyn_side_strap_03_l [{
   shapeType=Box
   extentX=5.3788
   extentY=6.4306
   extentZ=1.9278
   shapeName=dyn_side_strap_03_l_box
   center=<2.5939 0 0>
   rotation=<rot 0 -0.0002 9.2013>
  }]
 dyn_side_strap_03_r [{
   shapeType=Box
   extentX=5.3531
   extentY=7.0236
   extentZ=2.1342
   shapeName=dyn_side_strap_03_r_box
   center=<2.6998 0 0>
   rotation=<rot 0 0.0003 -3.5404>
  }]
 dyn_side_strap_04_l [{
   shapeType=Box
   extentX=8.2565
   extentY=6.2422
   extentZ=1.6441
   shapeName=dyn_side_strap_04_l_box
   center=<3.958 0 0>
   rotation=<rot 0 -0.0002 8.8542>
  }]
 dyn_side_strap_04_r [{shapeType=Box, extentX=8.3785, extentY=6.6285, extentZ=2.4701, shapeName=dyn_side_strap_04_r_box, center=<4.0057 0 0>}]
 pelvis [{shapeType=Capsule, length=9.9348, radius=11.5542, shapeName=pelvis_capsule, center=<-2.5255 -1.8239 0>}]
 thigh_l [{
   shapeType=Capsule
   length=19.0158
   radius=7.8053
   shapeName=thigh_l_capsule
   center=<-11.8177 -0.5057 1.151>
   rotation=<rot 89.4897 -27.3944 -19.1426>
  }]
 thigh_r [{
   shapeType=Capsule
   length=18.3648
   radius=8.2078
   shapeName=thigh_r_capsule
   center=<13.2952 0.3103 -1.5943>
   rotation=<rot 88.519 2.3731 9.674>
  }]
}
constraints [
 {bone1Name=dyn_side_strap_01_r, bone2Name=dyn_side_buckle_parent_r, swing1LimitDegrees=20, swing2LimitDegrees=20, twistLimitDegrees=3}
 {bone1Name=dyn_side_strap_02_r, bone2Name=dyn_side_strap_01_r, swing1LimitDegrees=20, swing2LimitDegrees=20, twistLimitDegrees=3}
 {bone1Name=dyn_side_strap_03_r, bone2Name=dyn_side_strap_02_r, swing1LimitDegrees=20, swing2LimitDegrees=20, twistLimitDegrees=3}
 {bone1Name=dyn_side_strap_04_r, bone2Name=dyn_side_strap_03_r, swing1LimitDegrees=20, swing2LimitDegrees=20, twistLimitDegrees=3}
 {bone1Name=dyn_side_strap_01_l, bone2Name=dyn_side_buckle_parent_l, swing1LimitDegrees=20, swing2LimitDegrees=20, twistLimitDegrees=3}
 {bone1Name=dyn_side_strap_04_l, bone2Name=dyn_side_strap_03_l, swing1LimitDegrees=20, swing2LimitDegrees=20, twistLimitDegrees=3}
 {bone1Name=dyn_side_strap_03_l, bone2Name=dyn_side_strap_02_l, swing1LimitDegrees=20, swing2LimitDegrees=20, twistLimitDegrees=3}
 {bone1Name=dyn_side_strap_02_l, bone2Name=dyn_side_strap_01_l, swing1LimitDegrees=20, swing2LimitDegrees=20, twistLimitDegrees=3}
]
