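# Post-process AnimBP for the female large Forsake body part, copying the parent mesh pose and adjusting its mechanical arm, floating rings, buckles, straps and kneepads.
# EventGraph computes sine/cosine bobbing offsets for the rings, plus alphas and flags from crouch, skydive, equipped weapon type and pistol aiming.
# These drive bone offsets; AnimDynamics adds ring springiness, weighted constraints anchor buckles and kneepads, RigidBody simulates straps, and a Control Rig finishes the pose.
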AnimBlueprint /BRCosmetics/Characters/Player/Female/Large/Bodies/F_LRG_Forsake/Meshes/F_LRG_Forsake_AnimBP
target_skeleton=&/F_LRG_Forsake_Skeleton
graphs [
 AnimGraph {nodes [
   n1 = Get Crouch_Alpha() -> (Crouch_Alpha:double)
   n2 = "Transform (Modify) Bone - Bone: dyn_mechanical_arm_l"(
    ComponentPose=Local To Component(LocalPose=Copy Pose From Mesh() {bUseAttachedParent=true})
    Alpha=Get Equipped_Alpha() -> (Equipped_Alpha:double)
    Translation="-1.5,0,2"
    Rotation="0.002,-19.784,4.439"
    Scale="1,1,1"
   ) {
    rotationMode=BMM_Additive
    rotationSpace=BCS_BoneSpace
    translationMode=BMM_Additive
    translationSpace=BCS_BoneSpace
    alphaBoolBlend {blendOption=Linear}
    alphaScaleBiasClamp {bInterpResult=true, interpSpeedDecreasing=20, interpSpeedIncreasing=20}
    boneToModify {boneName=dyn_mechanical_arm_l}
    rotation=<rot 0.002 -19.784 4.439>
    translation {x=-1.5, z=2}
   }
   n3 = Get Skydive_Alpha() -> (Skydive_Alpha:double)
   n4 = "Transform (Modify) Bone - Bone: dyn_mechanical_arm_l"(ComponentPose=@n2, Alpha=@n3, Translation="7.5,0,0.5", Rotation="-4,-16,-14", Scale="1,1,1") {
    rotationMode=BMM_Additive
    rotationSpace=BCS_BoneSpace
    translationMode=BMM_Additive
    translationSpace=BCS_BoneSpace
    alphaBoolBlend {blendOption=Linear}
    boneToModify {boneName=dyn_mechanical_arm_l}
    rotation=<rot -4 -16 -14>
    translation {x=7.5, z=0.5}
   }
   n5 = "Transform (Modify) Bone - Bone: dyn_arm_ring_parent_01"(ComponentPose=@n4, Alpha=@n3, Translation="-0.170,-9.858,0.027", Rotation="0,0,0", Scale="1,1,1") {
    rotationMode=BMM_Additive
    rotationSpace=BCS_BoneSpace
    translationMode=BMM_Additive
    translationSpace=BCS_BoneSpace
    alphaBoolBlend {blendOption=Linear}
    boneToModify {boneName=dyn_arm_ring_parent_01}
    translation=<-0.17 -9.858 0.027>
   }
   n6 = "Transform (Modify) Bone - Bone: dyn_arm_ring_parent_02"(ComponentPose=@n5, Alpha=@n3, Translation="0,-8,12", Rotation="-40,0,-8", Scale="1,1,1") {
    rotationMode=BMM_Additive
    rotationSpace=BCS_BoneSpace
    translationMode=BMM_Additive
    translationSpace=BCS_BoneSpace
    alphaBoolBlend {blendOption=Linear}
    boneToModify {boneName=dyn_arm_ring_parent_02}
    rotation {pitch=-40, roll=-8}
    translation {y=-8, z=12}
   }
   n7 = "Transform (Modify) Bone - Bone: dyn_arm_ring_base_02"!(ComponentPose=@n6, bAlphaBoolEnabled:bool=Get Alt_weapon() -> (Alt_weapon:bool), Translation="0,0,6", Rotation="10,0,0", Scale="1,1,1") -> (Pose) {
    alphaInputType=Bool
    rotationMode=BMM_Additive
    rotationSpace=BCS_BoneSpace
    translationMode=BMM_Additive
    translationSpace=BCS_BoneSpace
    alphaBoolBlend {blendInTime=0.15, blendOption=Linear, blendOutTime=0.15}
    boneToModify {boneName=dyn_arm_ring_base_02}
    rotation {pitch=10}
    translation {z=6}
   }
   n8 = "Transform (Modify) Bone - Bone: dyn_arm_ring_parent_01"(ComponentPose=@n7, Alpha=1, Translation=Get ArmRing_Float_A() -> (ArmRing_Float_A:Vector), Rotation="0,0,0", Scale="1,1,1") {
    translationMode=BMM_Additive
    translationSpace=BCS_BoneSpace
    alphaBoolBlend {blendOption=Linear}
    boneToModify {boneName=dyn_arm_ring_parent_01}
   }
   n9 = "Transform (Modify) Bone - Bone: dyn_arm_ring_parent_02"(ComponentPose=@n8, Alpha=1, Translation=Get ArmRing_Float_B() -> (ArmRing_Float_B:Vector), Rotation="0,0,0", Scale="1,1,1") {
    translationMode=BMM_Additive
    translationSpace=BCS_BoneSpace
    alphaBoolBlend {blendOption=Linear}
    boneToModify {boneName=dyn_arm_ring_parent_02}
   }
   n10 = Anim Dynamics - Bone: dyn_arm_ring_parent_02(ComponentPose=@n9, Alpha=0.2, ExternalForce="0,0,0") {
    angularDampingOverride=0.99
    angularSpringConstant=150
    bAngularSpring=true
    bGravityOverrideInSimSpace=true
    bLinearSpring=true
    bOverrideAngularDamping=true
    bOverrideLinearDamping=true
    gravityScale=0
    linearDampingOverride=0.9999
    linearSpringConstant=50
    simulationSpace=BoneRelative
    alphaBoolBlend {blendOption=Linear}
    boundBone {boneName=dyn_arm_ring_parent_02}
    componentLinearAccScale=<0.05 0.05 0.05>
    componentLinearVelScale=<0.8 1 0.6>
    physicsBodyDefinitions [{
      sphereCollisionRadius=0
      boundBone {boneName=dyn_arm_ring_parent_02}
      boxExtents=<50 50 50>
      constraintSetup {angularLimitsMax=<0 0 0>, angularLimitsMin=<0 0 0>, angularTarget=<0 0 0>, linearAxesMax=<2 5 5>, linearAxesMin=<-2 1 -5>}
      localJointOffset=<0 -3 0>
     }]
    relativeSpaceBone {boneName=spine_05}
   }
   n11 = Anim Dynamics - Bone: dyn_arm_ring_parent_01(ComponentPose=@n10, Alpha=0.2, ExternalForce="0,0,0") {
    angularDampingOverride=0.9
    angularSpringConstant=150
    bAngularSpring=true
    bGravityOverrideInSimSpace=true
    bLinearSpring=true
    bOverrideAngularDamping=true
    bOverrideLinearDamping=true
    gravityScale=0
    linearDampingOverride=0.995
    linearSpringConstant=50
    simulationSpace=BoneRelative
    alphaBoolBlend {blendOption=Linear}
    boundBone {boneName=dyn_arm_ring_parent_01}
    componentLinearAccScale=<0.1 0.1 0.3>
    componentLinearVelScale=<0.8 1 0.6>
    physicsBodyDefinitions [{
      sphereCollisionRadius=0
      boundBone {boneName=dyn_arm_ring_parent_01}
      boxExtents=<50 50 50>
      constraintSetup {angularLimitsMax=<0 0 0>, angularLimitsMin=<0 0 0>, angularTarget=<0 0 0>, linearAxesMax=<2 3 4>, linearAxesMin=<-2 1 -4>}
      localJointOffset=<0 -3 0>
     }]
    relativeSpaceBone {boneName=spine_05}
   }
   n12 = "Transform (Modify) Bone - Bone: dyn_side_buckle_parent_l"(ComponentPose=@n11, Alpha=@n1, Translation="-0.288,4.532,-0.24", Rotation="-6.440,-9.363,-19.445", Scale="1,1,1") {
    rotationMode=BMM_Additive
    rotationSpace=BCS_BoneSpace
    translationMode=BMM_Additive
    translationSpace=BCS_BoneSpace
    alphaBoolBlend {blendInTime=0.15, blendOption=Linear, blendOutTime=0.15}
    alphaScaleBiasClamp {bInterpResult=true, interpSpeedDecreasing=30, interpSpeedIncreasing=30}
    boneToModify {boneName=dyn_side_buckle_parent_l}
    rotation=<rot -6.44 -9.363 -19.445>
    translation=<-0.288 4.532 -0.24>
   }
   n13 = "Transform (Modify) Bone - Bone: dyn_side_buckle_parent_r"(ComponentPose=@n12, Alpha=@n1, Translation="0.636,-5.219,0.975", Rotation="-13.874,18.892,27.785", Scale="1,1,1") {
    rotationMode=BMM_Additive
    rotationSpace=BCS_BoneSpace
    translationMode=BMM_Additive
    translationSpace=BCS_BoneSpace
    alphaBoolBlend {blendInTime=0.15, blendOption=Linear, blendOutTime=0.15}
    alphaScaleBiasClamp {bInterpResult=true, interpSpeedDecreasing=30, interpSpeedIncreasing=30}
    boneToModify {boneName=dyn_side_buckle_parent_r}
    rotation=<rot -13.874 18.892 27.785>
    translation=<0.636 -5.219 0.975>
   }
   n14 = "Constraint - dyn_side_buckle_base_l (PPPPPP)"!(
    ComponentPose=@n13
    Alpha=1
    ConstraintWeights_0=0.1153
    ConstraintWeights_1:float=0.0472
    ConstraintWeights_2:float=0.0703
    ConstraintWeights_3:float=0.0577
    ConstraintWeights_4:float=0.0618
    ConstraintWeights_5:float=0.6478
   ) -> (Pose) {
    alphaBoolBlend {blendOption=Linear}
    boneToModify {boneName=dyn_side_buckle_base_l}
    constraintSetup [
     {targetBone {boneName=spine_01}}
     {targetBone {boneName=spine_02}}
     {targetBone {boneName=thigh_l}}
     {targetBone {boneName=spine_03}}
     {targetBone {boneName=thigh_twist_01_l}}
     {targetBone {boneName=pelvis}}
    ]
    constraintWeights [1,1,1,1,1,1]
   }
   n15 = "Constraint - dyn_side_buckle_base_r (PPPPPPP)"!(
    ComponentPose=@n14
    Alpha=1
    ConstraintWeights_0=0.0008
    ConstraintWeights_1:float=0.0607
    ConstraintWeights_2:float=0.0239
    ConstraintWeights_3:float=0.0005
    ConstraintWeights_4:float=0.0261
    ConstraintWeights_5:float=0.0791
    ConstraintWeights_6:float=0.8089
   ) -> (Pose) {
    alphaBoolBlend {blendOption=Linear}
    boneToModify {boneName=dyn_side_buckle_base_r}
    constraintSetup [
     {targetBone {boneName=spine_04}}
     {targetBone {boneName=spine_01}}
     {targetBone {boneName=spine_02}}
     {targetBone {boneName=thigh_r}}
     {targetBone {boneName=spine_03}}
     {targetBone {boneName=thigh_twist_01_r}}
     {targetBone {boneName=pelvis}}
    ]
    constraintWeights [1,1,1,1,1,1,1]
   }
   n16 = "Transform (Modify) Bone - Bone: dyn_side_buckle_base_l"(ComponentPose=@n15, Alpha=1, Translation="0.000,-0.000,-0.5", Rotation="2,-0.000,-0.000", Scale="1,1,1") {
    rotationMode=BMM_Additive
    rotationSpace=BCS_BoneSpace
    translationMode=BMM_Additive
    translationSpace=BCS_BoneSpace
    alphaBoolBlend {blendOption=Linear}
    boneToModify {boneName=dyn_side_buckle_base_l}
    rotation {pitch=2}
    translation {z=-0.5}
   }
   n17 = "Transform (Modify) Bone - Bone: dyn_side_buckle_base_r"(ComponentPose=@n16, Alpha=1, Translation="0,0,-0.5", Rotation="2,0,0", Scale="1,1,1") {
    rotationMode=BMM_Additive
    rotationSpace=BCS_BoneSpace
    translationMode=BMM_Additive
    translationSpace=BCS_BoneSpace
    alphaBoolBlend {blendOption=Linear}
    boneToModify {boneName=dyn_side_buckle_base_r}
    rotation {pitch=2}
    translation {z=-0.5}
   }
   n18 = RigidBody(ComponentPose=@n17) {
    bClampLinearTranslationLimitToRefPose=true
    bOverrideWorldGravity=true
    overridePhysicsAsset=&/F_LRG_Forsake_Straps_Physics
    simulationSpace=BaseBoneSpace
    alphaBoolBlend {blendOption=Linear}
    baseBoneRef {boneName=pelvis}
    componentAppliedLinearAccClamp=<6000 6000 6000>
    componentLinearAccScale=<0.54 0.4 0.4>
    componentLinearVelScale=<0.3 0.3 0.3>
    overrideWorldGravity {z=-980}
    simSpaceSettings {maxAngularAcceleration=1000, maxAngularVelocity=3, maxLinearAcceleration=250, maxLinearVelocity=500, velocityScaleZ=0, worldAlpha=1}
   }
   n19 = "Constraint - dyn_kneepad_base_r (P)"(ComponentPose=@n18, Alpha=1, ConstraintWeights_0=1) {
    alphaBoolBlend {blendOption=Linear}
    boneToModify {boneName=dyn_kneepad_base_r}
    constraintSetup [{targetBone {boneName=calf_r}}]
    constraintWeights [1]
   }
   n20 = "Constraint - dyn_kneepad_base_l (P)"(ComponentPose=@n19, Alpha=1, ConstraintWeights_0=1) {
    alphaBoolBlend {blendOption=Linear}
    boneToModify {boneName=dyn_kneepad_base_l}
    constraintSetup [{targetBone {boneName=calf_l}}]
    constraintWeights [1]
   }
   n21 = Control Rig(Source=Component To Local(ComponentPose=@n20)) {
    alphaBoolBlend {blendOption=Linear}
    controlRigAssetReference {blueprintRigClass=&/F_LRG_Forsake_CtrlRig.F_LRG_Forsake_CtrlRig_C}
   }
   Output Pose(Result=@n21) {layerGroup=None}
   Reroute Node(InputPin:Component Space Pose Link Structure=Reroute Node()) -> (OutputPin:Component Space Pose Link Structure)
  ]}
 EventGraph {nodes [
   Event BlueprintUpdateAnimation()
   > n1 = Sequence!(execute) -> (then_0, then_1, then_2:Exec, then_3:Exec, then_4:Exec, then_5:Exec, then_6:Exec)
   n2 = GetRealTimeSeconds()
   n3 = "Sin (Radians)"(A=float * float(A=float * float(A=1, B=1), B=@n2))
   n4 = "Sin (Radians)"(A=float * float(A=float * float(A=1, B=2), B=@n2))
   n5 = "Sin (Radians)"(A=float * float(A=float * float(A=1, B=1.5), B=@n2))
   > Set ArmRing_Float_A(ArmRing_Float_A:Vector=MakeVector(X=float * float(A=@n5, B=0.3), Y=float * float(A=@n3, B=0.56), Z=float * float(A=@n4, B=0.22))) -> (Output_Get:Vector)
   n6 = GetGameTimeInSeconds()
   n7 = "Cos (Radians)"!(A:double=float * float(A=float * float(A=1, B=1.5), B=@n6)) -> (ReturnValue:double)
   n8 = "Cos (Radians)"(A=float * float(A=float * float(A=1, B=1.78), B=@n6))
   n9 = "Cos (Radians)"(A=float * float(A=float * float(A=1, B=1.15), B=@n6))
   Set ArmRing_Float_B(
    execute=@n1.then_1
    ArmRing_Float_B:Vector=MakeVector(X=float * float(A=@n7, B=0.6), Y=float * float(A=@n8, B=0.725), Z=float * float(A=@n9, B=0.3))
   ) -> (Output_Get:Vector)
   Set Alt_weapon(execute=@n1.then_2, Alt_weapon:bool=false) -> (Output_Get:bool)
   > n10 = Switch on EFortWeaponCoreAnimation!(
    execute
    Selection:EFortWeaponCoreAnimation Enum=GetWeaponCoreAnim!(self:Fort Player Animation Interface Object Reference=GetOwnerAnimationInterface!(self:Custom Character Part Anim Instance Object Reference) -> (ReturnValue:Fort Player Animation Interface Object Reference)) -> (ReturnValue:EFortWeaponCoreAnimation Enum)
   ) -> (
    Melee:Exec
    Pistol:Exec
    Shotgun:Exec
    PaperBlueprint:Exec
    Rifle:Exec
    MeleeOneHand:Exec
    MachinePistol:Exec
    RocketLauncher:Exec
    GrenadeLauncher:Exec
    GoingCommando:Exec
    AssaultRifle:Exec
    TacticalShotgun:Exec
    SniperRifle:Exec
    TrapPlacement:Exec
    ShoulderLauncher:Exec
    AbilityDecoTool:Exec
    Crossbow:Exec
    C4:Exec
    RemoteControl:Exec
    DualWield:Exec
    AR_BullPup:Exec
    AR_ForwardGrip:Exec
    MedPackPaddles:Exec
    SMG_P90:Exec
    AR_DrumGun:Exec
    Consumable_Small:Exec
    Consumable_Large:Exec
    Balloon:Exec
    MountedTurret:Exec
    CreativeTool:Exec
    ExplosiveBow:Exec
    AshtonIndigo:Exec
    AshtonChicago:Exec
    MeleeDualWield:Exec
    Unarmed:Exec
   )
   Branch(execute=@n10.Pistol, Condition=Get bIsTargeting())
   > Set Alt_weapon(Alt_weapon=true)
   Set Weapon_Equipped(execute=@n1.then_5, Weapon_Equipped:bool=false) -> (Output_Get:bool)
   > n11 = Switch on EFortWeaponCoreAnimation(Selection=GetWeaponCoreAnim(self=GetOwnerAnimationInterface()))
   Set Weapon_Equipped(
    execute=[@n11.AssaultRifle, @n11.MachinePistol, @n11.Melee, @n11.Pistol, @n11.Rifle, @n11.Shotgun, @n11.SniperRifle, @n11.TacticalShotgun]
    Weapon_Equipped=true
   )
   Set Equipped_Alpha(execute=@n1.then_6, Equipped_Alpha:double=SelectFloat(A=1, B=0, bPickA=Get Weapon_Equipped() -> (Weapon_Equipped:bool))) -> (Output_Get:double)
   Set Crouch_Alpha(execute=@n1.then_3, Crouch_Alpha:double=SelectFloat(A=1, B=0, bPickA=Get bIsCrouching())) -> (Output_Get:double)
   Set Skydive_Alpha(execute=@n1.then_4, Skydive_Alpha:double=SelectFloat(A=1, B=0, bPickA=Get bIsSkyDiving())) -> (Output_Get:double)
   Get bIsTargeting()
  ]}
]
variables [
 Alt_weapon {type=boolean}
 ArmRing_Float_A {type=Vector}
 ArmRing_Float_B {type=Vector}
 Crouch_Alpha {type=number}
 Equipped_Alpha {type=number}
 Skydive_Alpha {type=number}
 Weapon_Equipped {type=boolean}
]
