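# Physics asset for the face accessory part of the large female Forsake player body, covering a ponytail bone chain.
# Capsules chain from a small base sphere through parent and segments 01 to 04, continuing to 06, while 05 branches off 03.
# Swing limits widen toward mid-chain and tighten at the tips, with tight twist limits throughout.
# Unconstrained capsules on neck_02 and spine_05 give neck and upper-torso collision shapes.
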PhysicsAsset /BRCosmetics/Characters/Player/Female/Large/Bodies/F_LRG_Forsake/Meshes/Parts/F_LRG_Forsake_FaceAcc_Physics
bodies {
 dyn_ponytail_01 [{shapeType=Capsule, length=1.7069, radius=3.3743, shapeName=dyn_ponytail_01_capsule, center=<3.3101 0 0>, rotation=<rot 90 0 0>}]
 dyn_ponytail_02 [{shapeType=Capsule, length=0.4767, radius=5.2304, shapeName=dyn_ponytail_02_capsule, center=<3.9141 0 0>, rotation=<rot 90 0 0>}]
 dyn_ponytail_03 [{shapeType=Capsule, length=1.3358, radius=6.229, shapeName=dyn_ponytail_03_capsule, center=<4.2771 0 0>, rotation=<rot 90 0 0>}]
 dyn_ponytail_04 [{shapeType=Capsule, length=8.7843, radius=3.1385, shapeName=dyn_ponytail_04_capsule, center=<6.5973 0 0>, rotation=<rot 90 0 0>}]
 dyn_ponytail_05 [{shapeType=Capsule, length=3.4893, radius=1.5074, shapeName=dyn_ponytail_05_capsule, center=<2.9475 0 0>, rotation=<rot 90 0 0>}]
 dyn_ponytail_06 [{shapeType=Capsule, length=4.1474, radius=1.6749, shapeName=dyn_ponytail_06_capsule, center=<2.4893 0 0>, rotation=<rot 90 0 0>}]
 dyn_ponytail_base [{shapeType=Sphere, radius=0.5, shapeName=dyn_ponytail_base_sphere}]
 dyn_ponytail_parent [{shapeType=Capsule, length=2.2787, radius=2.8707, shapeName=dyn_ponytail_parent_capsule, center=<3.4035 0 0>, rotation=<rot 90 0 0>}]
 neck_02 [{shapeType=Capsule, length=4.488, radius=6.5142, shapeName=neck_02_capsule, center=<0 -0.6376 0>, rotation=<rot 90 0 0>}]
 spine_05 [
  {shapeType=Capsule, length=20.3595, radius=9.6405, shapeName=spine_05_capsule, center=<4.075 0 0>}
  {
   shapeType=Capsule
   length=17
   radius=13
   shapeName=spine_05_capsule
   center=<-13.958 1.1753 -0.1315>
   rotation=<rot -89.1264 -10.0254 7.5615>
  }
 ]
}
constraints [
 {bone1Name=dyn_ponytail_parent, bone2Name=dyn_ponytail_base, swing1LimitDegrees=5, swing2LimitDegrees=5, twistLimitDegrees=3}
 {bone1Name=dyn_ponytail_01, bone2Name=dyn_ponytail_parent, swing1LimitDegrees=10, swing2LimitDegrees=10, twistLimitDegrees=5}
 {bone1Name=dyn_ponytail_02, bone2Name=dyn_ponytail_01, swing1LimitDegrees=15, swing2LimitDegrees=15, twistLimitDegrees=5}
 {bone1Name=dyn_ponytail_03, bone2Name=dyn_ponytail_02, swing1LimitDegrees=20, swing2LimitDegrees=20, twistLimitDegrees=5}
 {bone1Name=dyn_ponytail_04, bone2Name=dyn_ponytail_03, swing1LimitDegrees=20, swing2LimitDegrees=20, twistLimitDegrees=3}
 {bone1Name=dyn_ponytail_06, bone2Name=dyn_ponytail_04, swing1LimitDegrees=10, swing2LimitDegrees=10, twistLimitDegrees=3}
 {bone1Name=dyn_ponytail_05, bone2Name=dyn_ponytail_03, swing1LimitDegrees=10, swing2LimitDegrees=10, twistLimitDegrees=3}
]
