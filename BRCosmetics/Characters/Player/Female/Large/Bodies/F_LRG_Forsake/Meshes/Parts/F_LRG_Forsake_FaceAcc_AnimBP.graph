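# Animation Blueprint for the face accessory part of the F_LRG_Forsake player character, adding physics-driven secondary motion.
# It copies the attached parent mesh's pose into a cache feeding three spine_05-relative RigidBody simulations with separate physics assets.
# Backbling checks on customization load select the cape, pack or default simulation.
# Only the default simulation receives an upward external force, applied while skydive diving or floating.
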AnimBlueprint /BRCosmetics/Characters/Player/Female/Large/Bodies/F_LRG_Forsake/Meshes/Parts/F_LRG_Forsake_FaceAcc_AnimBP
target_skeleton=&/F_LRG_Forsake_FaceAcc_Skeleton
graphs [
 AnimGraph {nodes [
   n1 = RigidBody(ComponentPose=Local To Component(LocalPose=Use cached pose 'FaceAcc_Cache'())) {
    bClampLinearTranslationLimitToRefPose=true
    bOverrideWorldGravity=true
    overridePhysicsAsset=&/F_LRG_Forsake_FaceAcc_Pack_Physics
    simulationSpace=BaseBoneSpace
    alphaBoolBlend {blendOption=Linear}
    baseBoneRef {boneName=spine_05}
    componentAppliedLinearAccClamp=<6000 6000 6000>
    componentLinearAccScale=<0.1 0.1 0.2>
    componentLinearVelScale=<0.1 0.1 0.2>
    overrideWorldGravity {z=-980}
   }
   n2 = Component To Local(ComponentPose=@n1)
   n3 = RigidBody(ComponentPose=Local To Component(LocalPose=Use cached pose 'FaceAcc_Cache'())) {
    bClampLinearTranslationLimitToRefPose=true
    bOverrideWorldGravity=true
    overridePhysicsAsset=&/F_LRG_Forsake_FaceAcc_Cape_Physics
    simulationSpace=BaseBoneSpace
    alphaBoolBlend {blendOption=Linear}
    baseBoneRef {boneName=spine_05}
    componentAppliedLinearAccClamp=<6000 6000 6000>
    componentLinearAccScale=<0.1 0.1 0.2>
    componentLinearVelScale=<0.1 0.1 0.2>
    overrideWorldGravity {z=-980}
   }
   n4 = Component To Local(ComponentPose=@n3)
   n5 = Get isWearingCape() -> (isWearingCape:bool)
   n6 = RigidBody(ComponentPose=Local To Component(LocalPose=Use cached pose 'FaceAcc_Cache'()), ExternalForce=Get Velocity_Alpha() -> (Velocity_Alpha:Vector)) {
    bClampLinearTranslationLimitToRefPose=true
    bOverrideWorldGravity=true
    overridePhysicsAsset=&/F_LRG_Forsake_FaceAcc_Physics
    simulationSpace=BaseBoneSpace
    alphaBoolBlend {blendOption=Linear}
    baseBoneRef {boneName=spine_05}
    componentAppliedLinearAccClamp=<6000 6000 6000>
    componentLinearAccScale=<0.15 0.15 0.3>
    componentLinearVelScale=<0.15 0.15 0.3>
    overrideWorldGravity {z=-980}
   }
   Output Pose(Result=Blend Poses by bool(
     BlendPose_0=Blend Poses by bool(BlendPose_0=@n4, BlendPose_1=@n2, bActiveValue=@n5) {blendType=Linear}
     BlendPose_1=Component To Local(ComponentPose=@n6)
     bActiveValue=Get isWearingPackBool() -> (isWearingPackBool:bool)
    ) {blendType=Linear}) {layerGroup=None}
   Save cached pose 'FaceAcc_Cache'(Pose=Copy Pose From Mesh() {bUseAttachedParent=true}) {cacheName=FaceAcc_Cache}
  ]}
 EventGraph {nodes [
   Event BlueprintInitializeAnimation()
   > n1 = Cast To PlayerPawn_Athena(Object=TryGetPawnOwner())
   > n2 = Set Pawn Owner(Pawn Owner:Player Pawn Athena Object Reference=@n1) -> (Output_Get:Player Pawn Athena Object Reference)
   > Set Body_Mesh(Body_Mesh=GetSkeletalMeshForPartType(self=@n2))
   Event BlueprintUpdateAnimation()
   n3 = Get bIsSkydiveDiveMode()
   n4 = Get bIsSkydiveFloating()
   > Set Velocity_Alpha(Velocity_Alpha:Vector=SelectVector(A="0, 0,30000", B="0, 0, 0", bPickA=OR Boolean(A=@n3, B=@n4))) -> (Output_Get:Vector)
   Event OnCharacterCustomizationLoaded()
   n5 = CheckForBackbling()
   > Set isWearingPackBool(isWearingPackBool:bool=@n5) -> (Output_Get:bool)
   > Set isWearingCape(isWearingCape=AND Boolean(A=@n5, B=@n5.bIsBackblingLong))
  ]}
]
variables [
 Body_Mesh {type=SkeletalMeshComponent}
 Is Backbling Long {type=boolean}
 isWearingCape {type=boolean}
 isWearingPackBool {type=boolean}
 Pawn Owner {type=PlayerPawn_Athena_C}
 Velocity_Alpha {type=Vector}
]
